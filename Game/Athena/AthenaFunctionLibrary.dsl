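; Blueprint function library for Athena, bundling shared gameplay helpers for team checks, loot spawning, mesh extents, grounded height, effect contexts and math.
; The environmental knockback explosion sphere-overlaps actors and launches pawns with linear distance falloff and a minimum launch angle, then grants fall-damage immunity.
; Its Reverse input is stored but never read, and a valid SpecificActor is only stored, not launched.
; Loot helpers spawn pickups from item arrays or tier-group tables, returning those spawned.
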(fn AllOnSameTeam
  (params (Actor :type Actor%20Object%20Reference) (ActorArray :type Array%20of%20Actor%20Object%20References))
  (for _array_element ActorArray
    (if (not (Team|OnSameTeam Actor _array_element))
      (LocalVariables|Set_Result false)))
  (return (LocalVariables|Get_Result)))

(fn CreateEnvironmentalKnockbackExplosion
  (params (ExplosionSourcePoint :type Vector) (ExplosionRadius :type Float%20%28double-precision%29 :default 1.5e+03) (Intensity :type Float%20%28double-precision%29 :default 2.5e+03) (MinKnockbackAngle :type Float%20%28double-precision%29 :default 3e+01) (Reverse :type Boolean) (SpecificActor :type Actor%20Object%20Reference))
  (Utilities|FlowControl|Sequence
    (:then_0
      (LocalVariables|Set_SourcePoint ExplosionSourcePoint))
    (:then_1
      (LocalVariables|Set_Radius ExplosionRadius))
    (:then_2
      (LocalVariables|Set_Intensity Intensity))
    (:then_3
      (LocalVariables|Set_MinKnockbackAngle MinKnockbackAngle))
    (:then_4
      (LocalVariables|Set_Reverse Reverse))
    (:then_5
      (Utilities|IsValid SpecificActor
        (:Is%20Valid
          (LocalVariables|Set_CurrentActorBeingLaunched SpecificActor))
        (:Is%20Not%20Valid
          (bind _outactors (Collision|SphereOverlapActors 0 (LocalVariables|Get_SourcePoint) (LocalVariables|Get_Radius) (Utilities|Array|MakeArray "ObjectTypeQuery3" "ObjectTypeQuery7")))
          (for _array_element _outactors
            (LocalVariables|Set_CurrentActorBeingLaunched _array_element)
            (LocalVariables|SetVecToCurrentActor (- (Collision|GetActorBounds (LocalVariables|Get_CurrentActorBeingLaunched) true false) (LocalVariables|Get_SourcePoint)))
            (bind _asfort_pawn (Utilities|Casting|CastToFortPawn (LocalVariables|Get_CurrentActorBeingLaunched)))
            (Character|LaunchCharacter _asfort_pawn (* (Utilities|ClamptoMinAngleBetween (Math|Vector|Normalize (LocalVariables|GetVecToCurrentActor)) (Math|Vector|ProjectVectorontoPlane (Math|Vector|Normalize (LocalVariables|GetVecToCurrentActor)) (Transformation|GetActorUpVector (LocalVariables|Get_CurrentActorBeingLaunched))) (LocalVariables|Get_MinKnockbackAngle)) (* (- 1.0 (/ (Math|Vector|VectorLength (LocalVariables|GetVecToCurrentActor)) (LocalVariables|Get_Radius))) (LocalVariables|Get_Intensity))) true true)
            (Variables|/Script/FortniteGame.FortPawn|SetbIgnoreNextFallingDamage true _asfort_pawn)
            (Environment|ApplyGlobalEnvironmentGameplayEffecttoActor _asfort_pawn "/Game/Athena/Environments/Blueprints/DudeBro/GE_Athena_FallDmgImmune_FX.GE_Athena_FallDmgImmune_FX_C" 0 "(GameplayTags=)")))))))

(fn GetActorMeshExtents
  (params (Actor :type Actor%20Object%20Reference))
  (bind _returnvalue (Actor|GetComponentsByClass Actor "/Script/Engine.StaticMeshComponent"))
  (bind (_origin _boxextent) (Collision|GetActorBounds Actor true false))
  (bind (_currentmin_x _currentmin_y _currentmin_z) (Math|Vector|BreakVector (Variables|GetCurrentMin_X)))
  (bind (_currentmax_x _currentmax_y _currentmax_z) (Math|Vector|BreakVector (Variables|GetCurrentMax_X)))
  (if (> (Utilities|Array|Length _returnvalue) 0)
    (for _array_element _returnvalue
      (if (and (Rendering|IsVisible _array_element) (not (Variables|/Script/Engine.SceneComponent|GetbHiddenInGame _array_element)))
        (Utilities|FlowControl|Sequence
          (:then_0
            (bind (_returnvalue_min _returnvalue_max) (Utilities|Struct|BreakBox (StaticMesh|GetBoundingBox (Variables|/Script/Engine.StaticMeshComponent|GetStaticMesh _array_element))))
            (bind (_returnvalue_min_x _returnvalue_min_y _returnvalue_min_z) (Math|Vector|BreakVector _returnvalue_min))
            (bind (_returnvalue_max_x _returnvalue_max_y _returnvalue_max_z) (Math|Vector|BreakVector _returnvalue_max))
            (Variables|SetCurrentMin_X (Math|Vector|MakeVector (Math|Float|Min(Float) _currentmin_x _returnvalue_min_x) (Math|Float|Min(Float) _currentmin_y _returnvalue_min_y) (Math|Float|Min(Float) _currentmin_z _returnvalue_min_z))))
          (:then_1
            (Variables|SetCurrentMax_X (Math|Vector|MakeVector (Math|Float|Max(Float) _returnvalue_max_x _currentmax_x) (Math|Float|Max(Float) _returnvalue_max_y _currentmax_y) (Math|Float|Max(Float) _returnvalue_max_z _currentmax_z)))))))
    (if (Math|Vector|Equal(Vector) (LocalVariables|GetCurrentMax) "0, 0, 0" 0.0001)
      (return _boxextent)
      (else
        (return (- (LocalVariables|GetCurrentMax) (LocalVariables|GetCurrentMin)))))
    (else
      (return _boxextent))))

(fn GetPlayerGroundedZ
  (params (PlayerPawn :type Fort%20Player%20Pawn%20Object%20Reference))
  (return (- (.z (Transformation|GetActorLocation PlayerPawn)) (Variables|/Script/Engine.CapsuleComponent|GetCapsuleHalfHeight (Variables|/Script/Engine.Character|GetCapsuleComponent PlayerPawn)))))

(fn MakeSimpleFortEffectContext
  (params (Instigator :type Actor%20Object%20Reference) (Damage%20Source :type Actor%20Object%20Reference))
  (return (GameplayEffects|MakeFortEffectContext (Utilities|Casting|CastToFortAbilitySystemComponent (Ability|GetAbilitySystemComponent Instigator)) 0 0 Damage%20Source 0)))

(fn RowToBool
  (params (Input :type Scalable%20Float%20Structure%20%28by%20ref%29) (Level :type Float%20%28double-precision%29))
  (return (select (> (Math|Float|Floor (ScalableFloat|GetValueatLevel Input Level)) 0) true (Math|Conversions|ToBoolean(Integer) (Math|Float|Floor (ScalableFloat|GetValueatLevel Input Level))))))

(fn ShouldAllowDeathEffects
  (params (Damage%20Or%20Death%20Tags :type Gameplay%20Tag%20Container%20Structure))
  (return (GameplayTags|DoesContainerMatchTagQuery Damage%20Or%20Death%20Tags (LocalVariables|GetAllowDeathEffectsTagQuery))))

(fn SpawnLootFromArray
  (params (Array :type Array%20of%20Fort%20Item%20Entry%20Structures) (Position :type Vector) (Direction :type Vector) (OverrideMaxStackCount :type Integer) (bToss :type Boolean) (bRandomRotation :type Boolean) (bBlockedFromAutoPickup :type Boolean) (PickupInstigatorHandle :type Integer) (SourceType :type EFortPickupSourceTypeFlag%20Enum :default "Other") (Source :type EFortPickupSpawnSource%20Enum :default "Unset") (OptionalOwnerPC :type Fort%20Player%20Controller%20Object%20Reference) (bPickupOnlyRelevantToOwner :type Boolean))
  (Utilities|Array|Clear (LocalVariables|GetPickupsSpawned))
  (if true
    (for _array_element Array
      (bind (_array_element_itemdefinition _array_element_itemguid _array_element_count) (Class|FortItemFunctionLibrary|BreakFortItemEntry _array_element))
      (bind _asfort_world_item_definition (Utilities|Casting|CastToFortWorldItemDefinition _array_element_itemdefinition))
      (bind _returnvalue (Inventory|SpawnPickupInWorld 0 _asfort_world_item_definition _array_element_count Position Direction OverrideMaxStackCount bToss bRandomRotation bBlockedFromAutoPickup PickupInstigatorHandle SourceType Source OptionalOwnerPC bPickupOnlyRelevantToOwner))
      (Utilities|Array|AddUnique (LocalVariables|GetPickupsSpawned) _returnvalue))
    (return (LocalVariables|GetPickupsSpawned))
    (else
      (Development|PrintString 0 "NoLootPackageFound"))))

(fn SpawnLootFromTable
  (params (TierGroupName :type Name) (WorldLevel :type Integer :default 1) (ForcedLootTier :type Integer :default -1) (Position :type Vector) (Direction :type Vector) (OverrideMaxStackCount :type Integer) (bToss :type Boolean) (bRandomRotation :type Boolean) (bBlockedFromAutoPickup :type Boolean) (PickupInstigatorHandle :type Integer) (SourceType :type EFortPickupSourceTypeFlag%20Enum :default "Other") (Source :type EFortPickupSpawnSource%20Enum :default "Unset") (OptionalOwnerPC :type Fort%20Player%20Controller%20Object%20Reference) (bPickupOnlyRelevantToOwner :type Boolean))
  (Utilities|Array|Clear (LocalVariables|GetPickupsSpawned))
  (bind (_outloottodrop _returnvalue) (Loot|PickLootDrops TierGroupName WorldLevel ForcedLootTier "(GameplayTags=)"))
  (if _returnvalue
    (for _array_element _outloottodrop
      (bind (_array_element_itemdefinition _array_element_itemguid _array_element_count) (Class|FortItemFunctionLibrary|BreakFortItemEntry _array_element))
      (bind _asfort_world_item_definition (Utilities|Casting|CastToFortWorldItemDefinition _array_element_itemdefinition))
      (bind _returnvalue_2 (Inventory|SpawnPickupInWorld 0 _asfort_world_item_definition _array_element_count Position Direction OverrideMaxStackCount bToss bRandomRotation bBlockedFromAutoPickup PickupInstigatorHandle SourceType Source OptionalOwnerPC bPickupOnlyRelevantToOwner))
      (Utilities|Array|AddUnique (LocalVariables|GetPickupsSpawned) _returnvalue_2))
    (return (LocalVariables|GetPickupsSpawned))
    (else
      (Development|PrintString 0 "NoLootPackageFound"))))

(fn SpawnLootFromTableUsingItemEntry
  (params (TierGroupName :type Name) (WorldLevel :type Integer :default 1) (ForcedLootTier :type Integer :default -1) (Position :type Vector) (Direction :type Vector) (OverrideMaxStackCount :type Integer) (bToss :type Boolean) (bRandomRotation :type Boolean) (bBlockedFromAutoPickup :type Boolean) (PickupInstigatorHandle :type Integer) (SourceType :type EFortPickupSourceTypeFlag%20Enum :default "Other") (Source :type EFortPickupSpawnSource%20Enum :default "Unset") (OptionalOwnerPC :type Fort%20Player%20Controller%20Object%20Reference) (bPickupOnlyRelevantToOwner :type Boolean))
  (Utilities|Array|Clear (LocalVariables|GetPickupsSpawned))
  (bind (_outloottodrop _returnvalue) (Loot|PickLootDrops TierGroupName WorldLevel ForcedLootTier "(GameplayTags=)"))
  (if _returnvalue
    (for _array_element _outloottodrop
      (bind _returnvalue_2 (Inventory|SpawnPickupInWorldWithClassAndItemEntry 0 _array_element 0 Position Direction OverrideMaxStackCount bToss bRandomRotation bBlockedFromAutoPickup SourceType Source OptionalOwnerPC bPickupOnlyRelevantToOwner))
      (Utilities|Array|AddUnique (LocalVariables|GetPickupsSpawned) _returnvalue_2))
    (return (LocalVariables|GetPickupsSpawned))
    (else
      (Development|PrintString 0 "NoLootPackageFound"))))

(fn UnclampedLerp
  (params (A :type Float%20%28double-precision%29) (B :type Float%20%28double-precision%29) (Alpha :type Float%20%28double-precision%29))
  (return (+ A (* (- B A) Alpha))))
