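; Athena (Battle Royale) player controller Blueprint handling victory feedback, Creative hooks, item grants and squad marker setup.
; Win effects play the local victory stinger and choose camera lens effects and shakes from finisher, weapon and death cause.
; Dispatchers relay pawn possession to Creative devices and forward client-to-server RPCs on the server.
; Soft item references async-load when granted, and blade wielders drop the sword after a delay on unapproved interactions.
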(event EventBeginPlay
  ;[1 1 1] Allows Creative devices to know when the pawns are possessed during the player created games so they can react to that event.
  (Game|Player|BindEventtoOnPlayerPawnPossessed (delegate Custom|ReactToPawnPossession))
  (Class|FortAsyncActionCurrentPlaylistReady|PlaylistDataReadyAsync))

(event Custom|ReactToPawnPossession
  (params (PossessedPawn :type Pawn%20Object%20Reference))
  ;[1 1 1] Allows Creative devices to know when the pawns are possessed during the player created games so they can react to that event.
  (Default|CallPawnPossessedDispatcher PossessedPawn))

(event Custom|RPC_Event_ClientToServer
  (params (FortPlayerController :type Fort%20Player%20Controller%20Object%20Reference))
  (if (Networking|IsServer)
    (Default|CallRPCClienttoServerDispatcher FortPlayerController)))

(event EventPlayWinEffects
  (params (FinisherPawn :type Pawn%20Object%20Reference) (FinishingWeapon :type Fort%20Weapon%20Item%20Definition%20Object%20Reference) (DeathCause :type EDeathCause%20Enum :default "OutsideSafeZone") (bAudioOnly :type Boolean) (VictoryAnimation :type ERoundVictoryAnimation%20Enum :default "Default"))
  (bind (_asfort_weapon_ranged_dual _bsuccess) (Utilities|Casting|CastToFortWeaponRangedDualClass (WeaponItemDefinition|GetWeaponActorClass FinishingWeapon)))
  (CallFunction|HandleVictoryRoyaleAudio bAudioOnly)
  (if (Utilities|Enum|Equal(Enum) VictoryAnimation "Default")
    (if (and (Utilities|IsValid FinisherPawn :!class /Script/Engine.KismetSystemLibrary) (Utilities|IsValid FinishingWeapon :!class /Script/Engine.KismetSystemLibrary))
      (if (Math|Boolean|ANDBoolean (Utilities|IsValid FinisherPawn :!class /Script/Engine.KismetSystemLibrary) (select (Utilities|IsValid FinishingWeapon :!class /Script/Engine.KismetSystemLibrary) (or (Utilities|ClassIsChildOf (WeaponItemDefinition|GetWeaponActorClass FinishingWeapon) (Utilities|ResolveSoftReference (Variables|Default|GetB_Ranged_Generic_SoftClass))) _bsuccess) (CallFunction|IsCreativeMode)) (Utilities|Equal(Object) FinisherPawn (Pawn|GetControlledPawn)) (not (or (Utilities|Enum|Equal(Enum) DeathCause "RocketLauncher") (Utilities|Enum|Equal(Enum) DeathCause "GrenadeLauncher"))))
        (Game|Feedback|ClientSpawnGenericCameraLensEffect "/Game/Effects/Fort_Effects/Camera_Effects/B_CameraLens_Victory_Direct.B_CameraLens_Victory_Direct_C")
        (Game|Feedback|ClientStartCameraShake "/Game/Effects/Fort_Effects/Camera_Effects/BP_Camera_Shake_Pulse_Zoom.BP_Camera_Shake_Pulse_Zoom_C")
        (else
          _))
      (else
        _))
    (Game|Feedback|ClientSpawnGenericCameraLensEffect "/Game/Effects/Fort_Effects/Camera_Effects/B_CameraLens_Victory_Ally.B_CameraLens_Victory_Ally_C")
    (Game|Feedback|ClientStartCameraShake "/Game/Effects/Fort_Effects/Camera_Effects/BP_Camera_Shake_Pulse_Only.BP_Camera_Shake_Pulse_Only_C")))

(event EventEndPlay
  (params (EndPlayReason :type EEndPlayReason%20Enum :default "Destroyed"))
  (|CallParentFunction EndPlayReason)
  (Utilities|FlowControl|Sequence
    (:then_0
      (bind _victoryroyaleaudio (Variables|Default|GetVictoryRoyaleAudio)
        (:then
          (Audio|Components|Audio|Stop _victoryroyaleaudio))
        (:else)))
    (:then_1
      (bind _victoryroyalemusic1 (Variables|Default|GetVictoryRoyaleMusic1)
        (:then
          (Audio|Components|Audio|Stop _victoryroyalemusic1))
        (:else)))
    (:then_2
      (bind _victoryroyalemusic2 (Variables|Default|GetVictoryRoyaleMusic2)
        (:then
          (Audio|Components|Audio|Stop _victoryroyalemusic2))
        (:else)))))

(event Custom|ReplicateRotationScaleReset
  ;[1 0.06042 0.1201] Remove in 9.00, this just so that carmines changes to input rotation scale don't persist after death
  (Input|ResetInputRotationScale))

(event EventOnLeaveVolume
  (params (Volume :type Fort%20Volume%20Object%20Reference)))

(event EventOnEnterVolume
  (params (Volume :type Fort%20Volume%20Object%20Reference)))

(event Custom|GrantItem
  (params (Item :type Fort%20World%20Item%20Definition%20Soft%20Object%20Reference))
  ;[1 1 1] Permissions
  (if (Utilities|IsValidSoftObjectReference Item)
    (Inventory|GiveItemtoInventoryOwner self (Utilities|ResolveSoftReference Item))
    (else
      (bind _object (Utilities|AsyncLoadAsset Item)
        (:Completed
          (bind _asfort_world_item_definition (Utilities|Casting|CastToFortWorldItemDefinition _object))
          (Inventory|GiveItemtoInventoryOwner self _asfort_world_item_definition))))))

(event Custom|RevokeItem
  (params (Item :type Fort%20World%20Item%20Definition%20Soft%20Object%20Reference))
  ;[1 1 1] Permissions
  (Inventory|RemoveItemFromPlayer :PlayerController self :ItemDefinition (Utilities|ResolveSoftReference Item) :AmountToRemove -1 :bForceRemoval true :bSkipVerbRemoval false))

(event Custom|Handle%20Victory%20Royale%20Audio
  (params (Audio%20Only :type Boolean))
  (bind (_outvalue _returnvalue) (Fortnite|GetNumericCVar "Fort.PlayVictoryStingerFromPostGameScreenRoot"))
  ;[1 1 1] Victory royale audio logic
  (if (Utilities|Equal(Object) self (Fortnite|GetFirstLocalFortPlayerController true true))
    (if (not (and _returnvalue (> _outvalue 0)))
      (CallFunction|PlayVictoryStinger 0))))

(event EventOnLocalInteraction
  (params (ReceivingActor :type Actor%20Object%20Reference))
  ;[1 1 1] Drop sword on unapproved interaction (if it's equipped)
  (bind _asfort_player_pawn (Utilities|Casting|CastToFortPlayerPawn (Pawn|GetControlledPawn self)))
  (bind _asfort_pickup (Utilities|Casting|CastToFortPickup ReceivingActor))
  (if (GameplayTags|HasAnyMatchingGameplayTags _asfort_player_pawn (Variables|Default|GetBladeWielderTag))
    (if (not (GameplayTags|HasMatchingGameplayTag (Utilities|Casting|CastToFortWorldItemDefinition (Class|FortItemFunctionLibrary|BreakFortItemEntry (Pickup|GetItemEntry _asfort_pickup 0))) "(TagName=\"Item.InventoryService\")"))
      (Development|PrintString (Utilities|String|EnumtoString (Item|GetItemType (Utilities|Casting|CastToFortWorldItemDefinition (Class|FortItemFunctionLibrary|BreakFortItemEntry (Pickup|GetItemEntry _asfort_pickup 0))))))
      (Utilities|FlowControl|Delay 0.6)
      (CallFunction|RemoveItem)
      (else
        (Development|PrintString "Drop sword ignored --  picking up inventory service item")))))

(event Custom|RemoveItem
  ;[1 1 1] Drop sword on unapproved interaction (if it's equipped)
  (Inventory|DropSpecificItem "/Game/Athena/Items/Gameplay/Sword/AGID_Sword.AGID_Sword"))

(event EventPlayVictoryStinger
  (params (VictoryStingerCue :type Sound%20Cue%20Object%20Reference))
  ;[1 1 1] Victory royale audio logic
  ;;[1 0.7853 0.05729] Victory Stinger Logic
  (if (Utilities|IsValid VictoryStingerCue :!class /Script/Engine.KismetSystemLibrary)
    (bind _returnvalue (Audio|SpawnSound2D VictoryStingerCue))
    (else
      (Audio|SpawnSound2D (select (Utilities|IsValid (Class|AthenaHUDContext|GetVictoryStinger (BlueprintContext|GetAthenaHUDContext)) :!class /Script/Engine.KismetSystemLibrary) (Class|AthenaHUDContext|GetVictoryStinger (BlueprintContext|GetAthenaHUDContext)) "/Game/Athena/Sounds/Character/LastPlayerKilled_Cue.LastPlayerKilled_Cue"))))
  (Variables|Default|SetVictoryRoyaleAudio _returnvalue))

(fn FadeAudioComponents
  (params (NewMusicComponent :type Audio%20Component%20Object%20Reference) (PreviousMusicComponent :type Audio%20Component%20Object%20Reference) (Music :type Sound%20Base%20Object%20Reference) (StartTime :type Float%20%28double-precision%29))
  (Audio|Components|Audio|SetSound NewMusicComponent Music)
  (Utilities|IsValid NewMusicComponent
    (:Is%20Valid
      (Audio|Components|Audio|FadeIn NewMusicComponent 1.0 1.0 StartTime "Linear" :!class /Script/Engine.AudioComponent)
      (Utilities|IsValid PreviousMusicComponent
        (:Is%20Valid
          (Audio|Components|Audio|FadeOut PreviousMusicComponent 1.0 0.0 "Linear" :!class /Script/Engine.AudioComponent))))))

(fn GetCreativeOptionsBundleOverride
  (bind _returnvalue (DataRegistry|GetDataRegistryItem "FNE.CreativeOptionsBundle:MoveToolBundle" (Variables|Default|GetCreativeOptionBundleTableRow)))
  (if _returnvalue
    (return (Utilities|Struct|BreakFortCreativeOptionBundleTableRow (Variables|Default|GetCreativeOptionBundleTableRow)))
    (else
      (return (Variables|Creative|GetCreativeOptions)))))

(fn IsCreativeMode
  (return (or (Class|FortGameStateAthena|IsCreativeModePlaylist (Utilities|Casting|CastToFortGameStateAthena (Game|GetGameState))) (Class|FortGameStateAthena|IsCreativeLTMPlaylist (Utilities|Casting|CastToFortGameStateAthena (Game|GetGameState))))))

(fn SetupPermissionRecheckTimer)

(fn ConstructionScript
  (bind _returnvalue (CallFunction|GetMarkerComponent))
  (MarkerComponent|SetMarkerWidgetClass _returnvalue "/Game/Athena/HUD/Team/SquadMarker.SquadMarker_C")
  (MarkerComponent|SetMarkerActorClass _returnvalue (Variables|FortPlayerControllerAthena|GetSquadMarkerActorClass)))
