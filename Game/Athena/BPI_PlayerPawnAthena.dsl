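; BPI_PlayerPawnAthena is a Blueprint interface for the Athena player pawn, declaring functions with no logic of their own.
; It exposes getters for current weapon meshes, the death controller, falling and storm audio loop components, storm audio volume and storm aura particles.
; It also declares a rez-in event call and a warmup-state query.
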(fn CallRezInEvent)

(fn GetCurrentWeaponMeshComponents
  (return 0))

(fn GetFallingAudioLoopComponent
  (return 0))

(fn GetOnDeathController
  (return 0))

(fn GetStormAudioLoopComponent
  (return 0))

(fn GetStormAudioVolume
  (return 0.0))

(fn GetStormAuraParticlesComponent
  (return 0))

(fn IsInWarmup
  (return false))
