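; Blueprint interface for player search-and-hide interactions in Athena.
; GetOverrideSearchMontage returns a value that, by its name, supplies an alternate search montage.
; SetIsHidingInProp takes an IsHiding boolean to set whether the player is hiding in a prop.
; The file declares only these two function signatures, with no logic of its own.
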(fn GetOverrideSearchMontage
  (return 0))

(fn SetIsHidingInProp
  (params (IsHiding :type Boolean)))
