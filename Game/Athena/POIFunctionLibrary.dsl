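; POI function library with a server-only SetPOITextWithContextTag function that sets a POI's minimap text for given context tags.
; It finds the GSC_MiniMapDataOverride component on the Athena game state and calls OverrideLocations with the context tags and the POI tag wrapped in a container.
; Context tags must already be set up in that component's override array.
; A missing component or empty tags prints a warning instead.
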(fn SetPOITextWithContextTag
  (params (POI_Tag :type Gameplay%20Tag%20Structure) (Context_Tags :type Gameplay%20Tag%20Container%20Structure))
  (bind _worldcontext (Variables|Get__WorldContext))
  (bind _returnvalue (Fortnite|Athena|GameState|GetGameStateAthena _worldcontext))
  (bind _returnvalue_2 (Actor|GetComponentbyClass _returnvalue "/Game/Blueprints/Components/GSC_MiniMapDataOverride.GSC_MiniMapDataOverride_C"))
  (bind _returnvalue_3 (GameplayTags|MakeGameplayTagContainerfromTag POI_Tag))
  ;[1 1 1] Sets the text of a POI. Requires the POI tag and also to set up a context tag in GSC_MiniMapDataOverride in the Mini Map Data Overrides Array. SERVER ONLY!
  (if (Networking|IsServer _worldcontext)
    (Utilities|IsValid _returnvalue_2
      (:Is%20Valid
        (if (and (Utilities|GameplayTagContainerNotEmpty Context_Tags) (Utilities|GameplayTagContainerNotEmpty _returnvalue_3))
          (Class|FortGameStateComponentMiniMapDataOverride|OverrideLocations _returnvalue_2 Context_Tags _returnvalue_3)
          (else
            (Development|PrintString 0 "Unable to set override locations, Context or POI tags are empty" true true "(R=1.000000,G=1.000000,B=0.000000,A=1.000000)"))))
      (:Is%20Not%20Valid
        (Development|PrintString 0 "Unable to set override locations, Could not find minimap override component" true true "(R=1.000000,G=1.000000,B=0.000000,A=1.000000)")))))
