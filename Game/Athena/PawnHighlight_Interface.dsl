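; Blueprint interface for pawn highlighting, letting objects apply or clear a highlight on a pawn by source.
; AddHighlightFromSource takes a source object, a Pawn Highlight Structure and a flag to update only if that highlight is current.
; RemoveHighlightFromSource takes just the source, so each highlight is tied to the object that requested it.
; The file declares signatures only and holds no implementation.
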(fn AddHighlightFromSource
  (params (Source :type Object%20Reference) (Highlight :type Pawn%20Highlight%20Structure) (OnlyUpdateIfCurrent :type Boolean)))

(fn RemoveHighlightFromSource
  (params (Source :type Object%20Reference)))
