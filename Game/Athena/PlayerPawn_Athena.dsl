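; PlayerPawn_Athena is the Athena player pawn Blueprint, handling cosmetic and gameplay reactions for players.
; Safe-zone events choose storm cue tiers by phase damage, add or remove them locally, and drive a mobile post-process fade.
; Gameplay cue and pawn events drive DBNO, death, vitality animation, water, lobby res-in, Creative modes and melee swing dispatchers.
; Timers replicate aim rotation; a spectate-change handler clears stuck storm cues on simulated proxies.
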(event EventDestroyed
  (bind _menugoingcommandowweapon (Variables|Default|GetMenuGoingCommandowWeapon))
  ;[1 1 1] spawning the going commando gun in for the frontend menu screens
  (Utilities|IsValid _menugoingcommandowweapon
    (:Is%20Valid
      (Actor|DestroyActor _menugoingcommandowweapon))))

(event EventBeginPlay
  ;[1 1 1] spawning the going commando gun in for the frontend menu screens
  (Utilities|FlowControl|Sequence
    (:then_0
      ;;[1 1 1] used by the lobby due to some race conditions
      (Customization|BindEventtoOnFinishedCharacterCustomization (EventDispatchers|CreateEvent :!event Custom|OnCharacterCustomizationCompleted)))
    (:then_1
      (EventDispatchers|AssignOnDBNOStateChanged (EventDispatchers|CreateEvent :!event Custom|OnDBNOStateChanged_Event) :!class /Script/FortniteGame.FortPawn))
    (:then_2
      ;;[1 0 0] Give ability to test DBNO/Reboot revive
      (Variables|/Script/FortniteGame.FortActorComponent_MultiInteract|SetGradualCounterMultiplier (Variables|Default|GetEnableDecayRow) (Variables|Default|GetMultiInteractActorComponent))
      ;;;[0.06879 0.2844 0.211] Need to seperate these rows out
      (Variables|DBNO|SetDBNOHorizontalTetherMultiplier (Variables|Default|GetHF_TetherMultiplierHorz))
      (Variables|DBNO|SetDBNOVerticalTetherMultiplier (Variables|Default|GetHF_TetherMultiplierVert)))
    (:then_3
      ;; FORT-1016727: When you spectate someone in the storm, the outside-safe-zone GC is added to their sim proxy pawn, but does not get removed when you stop spectating, meaning when you are rebooted you will have storm FX on your screen until that pawn leaves the storm. This checks if the gameplay cue is active when we stop spectating the pawn and removes it
      (EventDispatchers|AssignOnIsLocallyViewedChanged (delegate Custom|OnIsLocallyViewedChanged_Event)))))

(event EventEndPlay
  (params (EndPlayReason :type EEndPlayReason%20Enum :default "Destroyed"))
  ;[1 1 1] spawning the going commando gun in for the frontend menu screens
  (Utilities|FlowControl|Sequence
    (:then_0
      (bind _skydiveaudioloop (Variables|ParachuteAudio|GetSkydiveAudioLoop)
        (:then
          (Audio|Components|Audio|FadeOut _skydiveaudioloop 0.5 0.0 "Linear" :!class /Script/Engine.AudioComponent) ; prevents the skydive loop from being stuck \non certain scenarios
        )
        (:else
          _))
      (if (Variables|Default|GetbWasEverInWarmUp)
        (bind _stormaudioloop_inst (Variables|Default|GetStormAudioLoop_Inst)
          (:then
            (Audio|Components|Audio|Stop _stormaudioloop_inst)
            (Components|DestroyComponent _stormaudioloop_inst)
            (Variables|Default|SetbStormAudioCleanedUp true))
          (:else))))
    (:then_1
      (EventDispatchers|UnbindEventfromOnIsLocallyViewedChanged (EventDispatchers|CreateEvent :!event Custom|OnIsLocallyViewedChanged_Event)))))

(event EventPossessed
  (params (NewController :type Controller%20Object%20Reference))
  (|CallParentFunction NewController)
  (Utilities|FlowControl|Delay 1)
  (if (GameplayTags|TagContainer|CompareTagAssetInterfacetoOtherTagContainers (Variables|Default|GetTC_FlakVest) self)
    (Ability|SendGameplayEventtoActor self (Variables|Default|GetEventSpawnEffect) (Utilities|Struct|MakeGameplayEventData :Instigator self :Target self :ContextHandle "()" :InstigatorTags "(GameplayTags=)" :TargetTags "(GameplayTags=)" :EventMagnitude 0.0 :TargetData "()"))
    (Development|ExecuteConsoleCommand "athena.rimlightintensity 1")))

(event Custom|OnDBNOStateChanged_Event_0
  (params (FortPawn :type Fort%20Pawn%20Object%20Reference) (bInIsDBNO :type Boolean)))

(event Custom|OnIsLocallyViewedChanged_Event_0
  (params (IsLocallyViewed :type Boolean)))

(event Custom|OnIsLocallyViewedChanged_Event
  (params (IsLocallyViewed :type Boolean))
  ;[1 1 1] spawning the going commando gun in for the frontend menu screens
  ;; FORT-1016727: When you spectate someone in the storm, the outside-safe-zone GC is added to their sim proxy pawn, but does not get removed when you stop spectating, meaning when you are rebooted you will have storm FX on your screen until that pawn leaves the storm. This checks if the gameplay cue is active when we stop spectating the pawn and removes it
  (if (Math|Boolean|ANDBoolean (not IsLocallyViewed) (Utilities|Enum|Equal(Enum) (Networking|GetLocalRole) "ROLE_SimulatedProxy") (Variables|Default|GetbOutsideSafeZoneGameplayCueActive))
    (CallFunction|RemoveSafeZoneGameplayCue)))

(event Custom|ApplyColdTag
  (bind _self self)
  (bind _returnvalue (Ability|GetAbilitySystemComponent _self))
  ;[0.6922 0.8771 1] ApplyColdTag
  (Utilities|IsValid _returnvalue
    (:Is%20Valid
      (GameplayEffects|ApplyGameplayEffectToSelf :self _returnvalue :GameplayEffectClass "/Game/Athena/Items/EnvironmentalItems/Player_Snow/GE_SnowFlakes_PlayerEffect.GE_SnowFlakes_PlayerEffect_C" :Level 0.0)
      (GameplayEffects|ApplyGameplayEffectToSelf :self _returnvalue :GameplayEffectClass "/Game/Athena/Items/EnvironmentalItems/Player_Snow/GE_ColdBreath_V2.GE_ColdBreath_V2_C" :Level 0.0))
    (:Is%20Not%20Valid
      (EventDispatchers|BindEventtoOnAbilitySystemComponentInitialized (delegate Custom|ApplyColdTag_Retry) :!class /Script/FortniteGame.FortPlayerPawn))))

(event Custom|SetupColdTag
  (bind _self self)
  (bind _returnvalue (Ability|GetAbilitySystemComponent _self))
  ;[0.6922 0.8771 1] ApplyColdTag
  (Utilities|FlowControl|DoOnce false
    (:Completed
      (Class|FortAsyncActionCurrentPlaylistReady|PlaylistDataReadyAsync
        (:OnReady
          ;[1 0 0.06124] Temporary check this to view snow
          (if false
            (Utilities|IsValid _returnvalue
              (:Is%20Valid
                (GameplayEffects|ApplyGameplayEffectToSelf :self _returnvalue :GameplayEffectClass "/Game/Athena/Items/EnvironmentalItems/Player_Snow/GE_SnowFlakes_PlayerEffect.GE_SnowFlakes_PlayerEffect_C" :Level 0.0)
                (GameplayEffects|ApplyGameplayEffectToSelf :self _returnvalue :GameplayEffectClass "/Game/Athena/Items/EnvironmentalItems/Player_Snow/GE_ColdBreath_V2.GE_ColdBreath_V2_C" :Level 0.0))
              (:Is%20Not%20Valid
                (EventDispatchers|BindEventtoOnAbilitySystemComponentInitialized (delegate Custom|ApplyColdTag_Retry) :!class /Script/FortniteGame.FortPlayerPawn)))))))))

(event WaterBody|EventOnEnteredWaterBody
  (params (WaterBody :type Fort%20Water%20Body%20Actor%20Object%20Reference) (bIsFirstBody :type Boolean))
  ;[0 0.27 0.9184] Procedural water effects - in development
  (Variables|VFX|SetIs%20First%20Water%20Body bIsFirstBody)
  (Variables|VFX|SetIsInWater true))

(event Custom|isHiddingInProp
  (params (isHidden :type Boolean))
  ;[0 0.3429 1] Removed need for per Asset Hide/UnHide check for CCPMs Updated with subfunction that avoids Weapon attachments and all non-CP based attachments - 3/22/21-EK
  (CallFunction|HideCharacterParts isHidden))

(event Custom|ApplyColdTag_Retry
  (params (AbilitySystemComponent :type Fort%20Ability%20System%20Component%20Object%20Reference) (AffectedPawn :type Fort%20Player%20Pawn%20Object%20Reference))
  (bind _self self)
  (bind _returnvalue (Ability|GetAbilitySystemComponent _self))
  ;[0.6922 0.8771 1] ApplyColdTag
  (GameplayEffects|ApplyGameplayEffectToSelf :self _returnvalue :GameplayEffectClass "/Game/Athena/Items/EnvironmentalItems/Player_Snow/GE_SnowFlakes_PlayerEffect.GE_SnowFlakes_PlayerEffect_C" :Level 0.0)
  (GameplayEffects|ApplyGameplayEffectToSelf :self _returnvalue :GameplayEffectClass "/Game/Athena/Items/EnvironmentalItems/Player_Snow/GE_ColdBreath_V2.GE_ColdBreath_V2_C" :Level 0.0))

(event WaterBody|EventOnExitedWaterBody
  (params (WaterBody :type Fort%20Water%20Body%20Actor%20Object%20Reference) (bIsLastBody :type Boolean))
  ;[0 0.27 0.9184] Procedural water effects - in development
  (Variables|Default|SetIs%20Last%20Water%20Body bIsLastBody)
  (Networking|SwitchHasAuthority
    (:Remote
      (Development|PrintString "ExitedWater" false false "(R=0.340762,G=0.000000,B=1.000000,A=1.000000)")
      (Variables|VFX|SetIsInWater false)
      (Networking|CanExecuteCosmeticEvents
        (:True
          ;;[1 1 1] Old Splash FX Added Athena check to stop it from spawning in boats.
          (if (not (FortGlobals|IsinAthena))
            (CallFunction|PlayWaterEnterExitAudioandFX (Transformation|GetWorldLocation (Variables|Character|GetCapsuleComponent)))))))))

(event Cosmetic|EventOnEnteredWaterVolume
  (params (WaterSurfaceLocation :type Vector%20%28by%20ref%29 :default "0, 0, 0"))
  ;[0.01415 0.3663 1] [CLIENT] AWaterVolume_Shallow enter/exit.  Play splash audio and fx if we're not in a vehicle.
  (CallFunction|OnEnteredorExitedWaterVolume true WaterSurfaceLocation))

(event Cosmetic|EventOnExitedWaterVolume
  (params (WaterSurfaceLocation :type Vector%20%28by%20ref%29 :default "0, 0, 0"))
  ;[0.01415 0.3663 1] [CLIENT] AWaterVolume_Shallow enter/exit.  Play splash audio and fx if we're not in a vehicle.
  (CallFunction|OnEnteredorExitedWaterVolume false WaterSurfaceLocation))

(event Custom|PlayWaterEnterExitAudioAndFX
  (params (WaterSurfaceLocation :type Vector))
  ;[0.01415 0.3663 1] HANDLE WATER - enter/exit splash & timed function for ambient FX determination, aligned for water volumes at the same level as the visual water
  (CallFunction|EnableWaterAudioNative (Variables|VFX|GetIsInWater))
  (Utilities|FlowControl|DoOnce false :!label _label
    (:Completed
      (if (not (Variables|FortPawn|GetbMovingEmoteSkipLandingFX))
        (Effects|SpawnEmitteratLocation "/Game/WaterAndWind/Particles/WaterInteraction/P_WaterSplash_athena.P_WaterSplash_Athena" WaterSurfaceLocation))
      (Utilities|FlowControl|Delay 0.4)
      (goto _label Reset))))

(event GameplayCue|GameplayCue.Abilities.Activation.Generic.HarvestBuff.Tier1
  (params (EventType :type EGameplayCueEvent%20Enum :default "OnActive") (Parameters :type Gameplay%20Cue%20Parameters%20Structure%20%28by%20ref%29))
  ;[1 1 1] Proto Team - Harvesting Improvement Test
  (switch Utilities|FlowControl|Switch|SwitchonEGameplayCueEvent EventType
    (:OnActive
      _)
    (:WhileActive
      _)
    (:Executed
      _)
    (:Removed
      _))
  (Ability|GameplayCue|ForwardGameplayCuetoParent))

(event BPI_PlayerSearchAndHide|EventSetIsHidingInProp
  (params (IsHiding :type Boolean))
  (CallFunction|IsHiddinginProp IsHiding))

(event Custom|OnEnteredOrExitedWaterVolume
  (params (bEntered :type Boolean) (WaterSurfaceLocation :type Vector))
  ;[0.01415 0.3663 1] [CLIENT] AWaterVolume_Shallow enter/exit.  Play splash audio and fx if we're not in a vehicle.
  (Variables|VFX|SetIsInWater bEntered)
  (if (not (Vehicle|IsinVehicle))
    (CallFunction|PlayWaterEnterExitAudioandFX WaterSurfaceLocation)))

(event GameplayCue|GameplayCue.Abilities.Activation.Generic.HarvestBuff.Tier2
  (params (EventType :type EGameplayCueEvent%20Enum :default "OnActive") (Parameters :type Gameplay%20Cue%20Parameters%20Structure%20%28by%20ref%29))
  ;[1 1 1] Proto Team - Harvesting Improvement Test
  (switch Utilities|FlowControl|Switch|SwitchonEGameplayCueEvent EventType
    (:OnActive
      _)
    (:WhileActive
      _)
    (:Executed
      _)
    (:Removed
      _))
  (Ability|GameplayCue|ForwardGameplayCuetoParent))

(event Custom|AddSafeZoneGameplayCue
  ;[1 1 1] Setup and Hande Outside SafeZone Gameplay Cue
  (Projectile|AddGameplayCue%28Looping%29Local self (Variables|Default|GetOutsideSafeZoneGameplayCueTag) (Ability|GameplayCue|MakeGameplayCueParameters :NormalizedMagnitude 0.0 :RawMagnitude 0.0 :EffectContext "()" :AggregatedSourceTags "(GameplayTags=)" :AggregatedTargetTags "(GameplayTags=)" :Location "0,0,0" :Normal "0,0,0" :GameplayEffectLevel (Variables|Default|GetSafeZonePhase) :AbilityLevel 1 :bReplicateLocationWhenUsingMinimalRepProxy false))
  (Variables|Default|SetbOutsideSafeZoneGameplayCueActive true)
  ;[1 1 1] Still have mobile hacks
  (CallFunction|AddStormFX))

(event Custom|SetSafeZoneTagWhenChanged
  (params (NewWhichSafeZoneTag :type Integer) (GameplayTag :type Gameplay%20Tag%20Structure) (bSpawnGameplayCue :type Boolean))
  ;[1 1 1] Is storm Cn
  (bind _asfort_game_state_athena (Utilities|Casting|CastToFortGameStateAthena (Game|GetGameState)))
  (if (Class|FortGameStateAthena|IsLastManStandingCN _asfort_game_state_athena)
    _
    (elif (Variables|Default|GetbOutsideSafeZoneGameplayCueActive)
      ;[1 1 1] Setup and Hande Outside SafeZone Gameplay Cue
      (if (!= (Variables|Default|GetwhichSafeZoneTag) NewWhichSafeZoneTag)
        (CallFunction|RemoveSafeZoneGameplayCue))))
  (Variables|Default|SetwhichSafeZoneTag NewWhichSafeZoneTag)
  (Variables|Default|SetOutsideSafeZoneGameplayCueTag GameplayTag)
  (if bSpawnGameplayCue
    (Projectile|AddGameplayCue%28Looping%29Local self (Variables|Default|GetOutsideSafeZoneGameplayCueTag) (Ability|GameplayCue|MakeGameplayCueParameters :NormalizedMagnitude 0.0 :RawMagnitude 0.0 :EffectContext "()" :AggregatedSourceTags "(GameplayTags=)" :AggregatedTargetTags "(GameplayTags=)" :Location "0,0,0" :Normal "0,0,0" :GameplayEffectLevel (Variables|Default|GetSafeZonePhase) :AbilityLevel 1 :bReplicateLocationWhenUsingMinimalRepProxy false))
    (Variables|Default|SetbOutsideSafeZoneGameplayCueActive true)
    ;[1 1 1] Still have mobile hacks
    (CallFunction|AddStormFX)))

(event Custom|SetSafeZonePhaseAndGameplayCueTag
  (params (InSafeZonePhase :type Integer) (bSpawnGameplayCue :type Boolean))
  (bind _safezonedamagescalablefloat (Variables|Default|GetSafeZoneDamageScalableFloat))
  ;[1 1 1] Setup and Hande Outside SafeZone Gameplay Cue
  (bind _output_get (Variables|Default|SetSafeZonePhase InSafeZonePhase))
  (if (<= (ScalableFloat|GetValueatLevel _safezonedamagescalablefloat (Math|Conversions|ToFloat(Integer) _output_get)) 0.033333)
    (CallFunction|SetSafeZoneTagwhenChanged 0 (Variables|Default|GetOutsideSafeZoneCue) bSpawnGameplayCue)
    (elif (< (ScalableFloat|GetValueatLevel _safezonedamagescalablefloat (Math|Conversions|ToFloat(Integer) _output_get)) 0.09)
      (CallFunction|SetSafeZoneTagwhenChanged 1 (Variables|Default|GetOutsideSafeZone2Cue) bSpawnGameplayCue)
      (else
        (CallFunction|SetSafeZoneTagwhenChanged 2 (Variables|Default|GetOutsideSafeZone3Cue) bSpawnGameplayCue)))))

(event EventOnSuccessfulBuildingEdit
  ;[1 1 1] Sound played when an enemy confirms an edit
  (Audio|PlaySoundatLocation "/Game/Sounds/Fort_UI_Sounds/Gameplay/Fort_Build_BluePrint_EnemyConfirmEdit_Cue.Fort_Build_BluePrint_EnemyConfirmEdit_Cue" (Transformation|GetWorldLocation (Variables|Character|GetMesh))))

(event Custom|RemoveSafeZoneGameplayCue
  ;[1 1 1] Setup and Hande Outside SafeZone Gameplay Cue
  (Variables|Default|SetbOutsideSafeZoneGameplayCueActive false)
  (Projectile|RemoveGameplayCue%28Looping%29Local self (Variables|Default|GetOutsideSafeZoneGameplayCueTag) (Ability|GameplayCue|MakeGameplayCueParameters :NormalizedMagnitude 0.0 :RawMagnitude 0.0 :EffectContext "()" :AggregatedSourceTags "(GameplayTags=)" :AggregatedTargetTags "(GameplayTags=)" :Location "0,0,0" :Normal "0,0,0" :GameplayEffectLevel 1 :AbilityLevel 1 :bReplicateLocationWhenUsingMinimalRepProxy false))
  ;[1 1 1] Still have mobile hacks
  (CallFunction|RemoveStormFX))

(event Custom|RemoveSafeZoneGameplayCueServerToClient
  ;[1 1 1] Setup and Hande Outside SafeZone Gameplay Cue
  (Variables|Default|SetbOutsideSafeZoneGameplayCueActive false)
  (Projectile|RemoveGameplayCue%28Looping%29Local self (Variables|Default|GetOutsideSafeZoneGameplayCueTag) (Ability|GameplayCue|MakeGameplayCueParameters :NormalizedMagnitude 0.0 :RawMagnitude 0.0 :EffectContext "()" :AggregatedSourceTags "(GameplayTags=)" :AggregatedTargetTags "(GameplayTags=)" :Location "0,0,0" :Normal "0,0,0" :GameplayEffectLevel 1 :AbilityLevel 1 :bReplicateLocationWhenUsingMinimalRepProxy false))
  ;[1 1 1] Still have mobile hacks
  (CallFunction|RemoveStormFX))

(event Gameplay|EventNotifyTeammateSkydivedFromBus
  ;[1 1 1] Sound played when teammates jump out of the bus
  (Audio|SpawnSoundAttached "/Game/Athena/Sounds/Bus/FlyingBus_Drop_Teammate_Cue.FlyingBus_Drop_Teammate_Cue" (Variables|Transformation|GetRootComponent)))

(event Gameplay|EventOnOutsideSafeZonePhaseChanged
  (params (SafeZonePhase :type Integer))
  ;[1 1 1] Setup and Hande Outside SafeZone Gameplay Cue
  (CallFunction|SetSafeZonePhaseandGameplayCueTag SafeZonePhase (Variables|Gameplay|GetbIsInAnyStorm)))

(event WaterBody|EventOnBeginWaterSprintBoost
  (if (Pawn|IsLocallyControlled)
    (GameplayCue|ExecuteGameplayCue%28Burst%29Local self (Variables|Default|GetWaterSprintBoostTag) (Ability|GameplayCue|MakeGameplayCueParameters 0.0))))

(event Gameplay|EventSafeZoneStatusChanged
  (bind _returnvalue (SafeZone|GetSafeZoneIndicator))
  (bind _bisinanystorm (Variables|Gameplay|GetbIsInAnyStorm))
  ;[1 1 1] FX/Audio for outside safe zone (todo)
  (Networking|CanExecuteCosmeticEvents
    (:True
      (if (Math|Boolean|NotEqual(Boolean) _bisinanystorm (Variables|Default|GetbIsOutsideSafeZoneCached))
        (Variables|Default|SetbIsOutsideSafeZoneCached _bisinanystorm)
        (Audio|SpawnSoundAttached (Variables|AthenaAudio|GetSafeZonePassThroughSound) (Variables|Character|GetMesh)))
      (if (Variables|Gameplay|GetbIsInAnyStorm)
        (Utilities|IsValid _returnvalue
          (:Is%20Valid
            (CallFunction|SetSafeZonePhaseandGameplayCueTag (SafeZone|GetCurrentPhaseIndex _returnvalue) true))
          (:Is%20Not%20Valid
            (CallFunction|SetSafeZonePhaseandGameplayCueTag (Variables|Default|GetSafeZonePhase) true)))
        (else
          (CallFunction|RemoveSafeZoneGameplayCue))))))

(event GameplayCue|GameplayCue.Athena.Equipping
  (params (EventType :type EGameplayCueEvent%20Enum :default "OnActive") (Parameters :type Gameplay%20Cue%20Parameters%20Structure%20%28by%20ref%29))
  (bind _self self)
  (bind _returnvalue (Vehicle|GetVehicleInterface _self))
  ;[1 1 1] Cue For User Using Medkit/Bag/Sheild
  (Utilities|IsValid (Utilities|ToObject(Interface) _returnvalue)
    (:Is%20Valid
      (bind _asfort_character_vehicle (Utilities|Casting|CastToFortCharacterVehicle (Vehicle|GetActor _returnvalue))
        (:then
          (if (Utilities|Equal(Object) (Vehicle|GetDriver _returnvalue) self)
            (Variables|Default|SetPlayerPawnAthena _asfort_character_vehicle)
            (else
              (Variables|Default|SetPlayerPawnAthena self :!label _label)))
          (if (not (Pawn|IsLocallyControlled (Variables|Default|GetPlayerPawnAthena)))
            (Utilities|Casting|CastToFortPlayerControllerSpectating (Pawn|GetLocalViewingPlayerController (Variables|Default|GetPlayerPawnAthena))))
          (EventDispatchers|CallOnUIGameplayCue (Variables|Default|GetPlayerPawnAthena) "GameplayCue.Athena.Equipping" EventType Parameters))
        (:CastFailed
          _)))
    (:Is%20Not%20Valid
      _))
  (goto _label))

(event GameplayCue|GameplayCue.Athena.Boost
  (params (EventType :type EGameplayCueEvent%20Enum :default "OnActive") (Parameters :type Gameplay%20Cue%20Parameters%20Structure%20%28by%20ref%29))
  ;[1 1 1] Comment
  (if (not (Pawn|IsLocallyControlled))
    (Utilities|Casting|CastToFortPlayerControllerSpectating (Pawn|GetLocalViewingPlayerController)))
  (EventDispatchers|CallOnUIGameplayCue "GameplayCue.Athena.Boost" EventType Parameters))

(event GameplayCue|GameplayCue.Athena.Reloading
  (params (EventType :type EGameplayCueEvent%20Enum :default "OnActive") (Parameters :type Gameplay%20Cue%20Parameters%20Structure%20%28by%20ref%29))
  ;[1 1 1] Cur for Vehicle Reloading
  (if (not (Pawn|IsLocallyControlled))
    (Utilities|Casting|CastToFortPlayerControllerSpectating (Pawn|GetLocalViewingPlayerController)))
  (EventDispatchers|CallOnUIGameplayCue "GameplayCue.Athena.Reloading" EventType Parameters))

(event EventOnLandscapeCheckFailedAndTeleported
  (GameplayEffects|ApplyGameplayEffectToSelf :self (Ability|GetAbilitySystemComponent self) :GameplayEffectClass "/Game/Athena/MappedEffects/Lockout/GE_Athena_Lockout.GE_Athena_Lockout_C" :Level 0.0))

(event GameplayCue|GameplayCue.Athena.Player.BeingRevivedFromDBNO
  (params (EventType :type EGameplayCueEvent%20Enum :default "OnActive") (Parameters :type Gameplay%20Cue%20Parameters%20Structure%20%28by%20ref%29))
  ;[1 1 1] Cue For User Being Revived
  (if (not (Pawn|IsLocallyControlled))
    (Utilities|Casting|CastToFortPlayerControllerSpectating (Pawn|GetLocalViewingPlayerController)))
  (EventDispatchers|CallOnUIGameplayCue "GameplayCue.Athena.Player.BeingRevivedFromDBNO" EventType Parameters))

(event Custom|ActivateInvincibleVisuals
  (params (Duration :type Float%20%28double-precision%29))
  ;[1 1 1] DBNO Revive Invulnerable Visuals
  (Networking|CanExecuteCosmeticEvents
    (:True
      (if (> Duration 0.0)
        (PawnHighlight|HighlightPawnFromSource self (Variables|PawnHighlight|GetInvulnerable%20Highlight))
        (Utilities|FlowControl|RetriggerableDelay Duration)
        (PawnHighlight|RemovePawnHighlightFromSource self)))))

(event GameplayCue|GameplayCue.Abilities.Activation.DBNOResurrect
  (params (EventType :type EGameplayCueEvent%20Enum :default "OnActive") (Parameters :type Gameplay%20Cue%20Parameters%20Structure%20%28by%20ref%29))
  ;[1 1 1] DBNO Revive Invulnerable Visuals
  (CallFunction|ActivateInvincibleVisuals (Ability|GameplayCue|BreakGameplayCueParameters Parameters)))

(event EventOnTetheredExitPressed
  ;[1 1 1] Signal the UI that we're holding a button to exit tethering
  (Ability|SendGameplayEventtoActor self "(TagName=\"Athena.Vehicle.Biplane.OnHoldExitStarted\")" (Utilities|Struct|MakeGameplayEventData :ContextHandle "()" :InstigatorTags "(GameplayTags=)" :TargetTags "(GameplayTags=)" :EventMagnitude (Variables|/Script/FortniteGame.FortPawnComponent_Tether|GetTetheredExitHoldTime (Variables|FortPlayerPawn|GetTetherComponent)) :TargetData "()")))

(event GameplayCue|GameplayCue.Abilities.Activation.DBNOEnter
  (params (EventType :type EGameplayCueEvent%20Enum :default "OnActive") (Parameters :type Gameplay%20Cue%20Parameters%20Structure%20%28by%20ref%29))
  ;[1 1 1] Cue for player going DBNO
  (switch Utilities|FlowControl|Switch|SwitchonEGameplayCueEvent EventType
    (:Executed
      (CallFunction|ActivateInvincibleVisuals (CallFunction|GetDownedDBNOInvulnerableTime)) ; TODO: This Time value should be data driven
    )))

(event EventOnTetheredExitReleased
  ;[1 1 1] Signal the UI that we're holding a button to exit tethering
  (Ability|SendGameplayEventtoActor self "(TagName=\"Athena.Vehicle.Biplane.OnHoldExitStopped\")"))

(event GameplayCue|GameplayCue.Damage
  (params (EventType :type EGameplayCueEvent%20Enum :default "OnActive") (Parameters :type Gameplay%20Cue%20Parameters%20Structure%20%28by%20ref%29))
  ;[1 1 1] Cue for Taking Damage
  (Ability|GameplayCue|ForwardGameplayCuetoParent)
  (bind _returnvalue (Class|FortAnimInputEvent|CreateAnimInputEvent "aie.PlayerVitality.HealthDecreased"))
  (Animation|TriggerAnimInputEvent _returnvalue))

(event GameplayCue|GameplayCue.Shield.PotionConsumed
  (params (EventType :type EGameplayCueEvent%20Enum :default "OnActive") (Parameters :type Gameplay%20Cue%20Parameters%20Structure%20%28by%20ref%29))
  ;[1 1 1] Cue for Comsuming Shield Potion
  (Ability|GameplayCue|ForwardGameplayCuetoParent)
  (bind _returnvalue (Class|FortAnimInputEvent|CreateAnimInputEvent "aie.PlayerVitality.ShieldsIncreased"))
  (Animation|TriggerAnimInputEvent _returnvalue))

(event Custom|OnDBNOStateChanged_Event
  (params (FortPawn :type Fort%20Pawn%20Object%20Reference) (bInIsDBNO :type Boolean))
  (bind _multiinteractactorcomponent (Variables|Default|GetMultiInteractActorComponent))
  ;[1 1 1] Cue for player going DBNO
  (if bInIsDBNO
    (SoundLibrary|AddLibrary (Variables|SoundLibrary|GetSoundLibraryComponent) "/Game/Athena/Sounds/Character/SL_DBNO_Crawl.SL_DBNO_Crawl" :!class /Script/SoundLibrary.SoundLibraryComponent)
    (switch Utilities|FlowControl|Switch|SwitchonEFortTeamAffiliation (Team|GetTeamAffiliationtoLocalPlayer (Team|GetActorTeam FortPawn) true true)
      (:Friendly
        (if (CallFunction|GetPawnIsinLocalSquad)
          (Audio|PlaySoundatLocation (Variables|Default|GetDBNOEnterSound) (Transformation|GetActorLocation)))))
    (else
      (CallFunction|ResetCachedDBNOInvulnerableTime)
      (SoundLibrary|RemoveLibrary (Variables|SoundLibrary|GetSoundLibraryComponent) "/Game/Athena/Sounds/Character/SL_DBNO_Crawl.SL_DBNO_Crawl" :!class /Script/SoundLibrary.SoundLibraryComponent)
      (Utilities|IsValid _multiinteractactorcomponent
        (:Is%20Valid
          (Class|FortActorComponentMultiInteract|ResetGradualInteractionTimer _multiinteractactorcomponent))))))

(event GameplayCue|GameplayCue.Healing
  (params (EventType :type EGameplayCueEvent%20Enum :default "OnActive") (Parameters :type Gameplay%20Cue%20Parameters%20Structure%20%28by%20ref%29))
  ;[1 1 1] Cue for Healing
  (Ability|GameplayCue|ForwardGameplayCuetoParent)
  (bind _returnvalue (Class|FortAnimInputEvent|CreateAnimInputEvent "aie.PlayerVitality.HealthIncreased"))
  (Animation|TriggerAnimInputEvent _returnvalue))

(event GameplayCue|GameplayCue.Damage.Shielded
  (params (EventType :type EGameplayCueEvent%20Enum :default "OnActive") (Parameters :type Gameplay%20Cue%20Parameters%20Structure%20%28by%20ref%29))
  ;[1 1 1] Cue for Damage with Shields still intact
  (Ability|GameplayCue|ForwardGameplayCuetoParent)
  (bind _returnvalue (Class|FortAnimInputEvent|CreateAnimInputEvent "aie.PlayerVitality.ShieldsDecreased"))
  (Animation|TriggerAnimInputEvent _returnvalue))

(event EventOnDeathPlayEffects
  (params (Damage :type Float%20%28single-precision%29) (DamageTags :type Gameplay%20Tag%20Container%20Structure%20%28by%20ref%29) (Momentum :type Vector :default "0, 0, 0") (HitInfo :type Hit%20Result%20Structure%20%28by%20ref%29) (InstigatedBy :type Fort%20Pawn%20Object%20Reference) (DamageCauser :type Actor%20Object%20Reference) (EffectContext :type Gameplay%20Effect%20Context%20Handle%20Structure))
  ;[1 1 1] On player killed
  (|CallParentFunction Damage DamageTags Momentum HitInfo InstigatedBy DamageCauser EffectContext)
  (Collision|SetActorEnableCollisionwithChildren 0)
  (CallFunction|PlayDeathEffects))

(event EventOnDeathServer
  (params (Damage :type Float%20%28single-precision%29) (DamageTags :type Gameplay%20Tag%20Container%20Structure%20%28by%20ref%29) (Momentum :type Vector :default "0, 0, 0") (HitInfo :type Hit%20Result%20Structure%20%28by%20ref%29) (InstigatedBy :type Controller%20Object%20Reference) (DamageCauser :type Actor%20Object%20Reference) (EffectContext :type Gameplay%20Effect%20Context%20Handle%20Structure))
  ;[1 1 1] On player killed
  (|CallParentFunction Damage DamageTags Momentum HitInfo InstigatedBy DamageCauser EffectContext)
  (Collision|SetActorEnableCollisionwithChildren 0)
  (Variables|Elimination|SetOn%20Death%20Damage%20Tags DamageTags)
  (Variables|Elimination|SetOn%20Death%20Momentum Momentum)
  (Variables|Elimination|SetOn%20Death%20Effect%20Context EffectContext)
  (Variables|Elimination|SetOn%20Death%20Controller (Pawn|GetController self)))

(event Custom|AimTick
  ;[1 1 1] Special Tick for Replicating the Control rotation of the player to other clients. Used By HighTower Tomato Repulsor Cannon
  (Utilities|FlowControl|Gate true
    (:Exit
      (Networking|ForceNetUpdate)
      (CallFunction|ReplicateAimRotation (Pawn|GetControlRotation)))))

(event Custom|StartAImRotationTick
  ;[1 1 1] Special Tick for Replicating the Control rotation of the player to other clients. Used By HighTower Tomato Repulsor Cannon
  ;;[1 1 1] Safeguard timer to close tick gate
  (Utilities|Time|SetTimerbyFunctionName 0 "EndRotationTick" 2.5)
  (bind _returnvalue (Utilities|Time|SetTimerbyFunctionName 0 "AimTick" (ScalableFloat|GetValueatLevel (Variables|Default|GetControlRotationReplicationTickRate)) true))
  (CallFunction|AimTick)
  (Utilities|FlowControl|Gate true
    (:Exit
      (Networking|ForceNetUpdate)
      (CallFunction|ReplicateAimRotation (Pawn|GetControlRotation)))))

(event Custom|ReplicateAimRotation
  (params (PlayerAimRotation :type Rotator))
  ;[1 1 1] Special Tick for Replicating the Control rotation of the player to other clients. Used By HighTower Tomato Repulsor Cannon
  (Variables|Default|SetPlayerAimRotation PlayerAimRotation))

(event Custom|AddStormFX
  ;[1 1 1] Comment
  ;;[1 1 1] Events for managing storm-related fx, most things moved to GC_Athena_OutsideSafeZone
  ;;;[1 0 0] Mobile Only Post Settings  STRIP if ConstructScriptWorks
  (if (and (Fortnite|IsMobilePlatform) (Pawn|IsLocallyControlled))
    (bind _returnvalue (AddComponent|NODE_AddPostProcessComponent-2))
    (Variables|Default|SetMobilePostProcess _returnvalue)
    (bind (_direction _blendweight) (Timeline|Timeline_0 Play)
      (:Update
        (Variables|/Script/Engine.PostProcessComponent|SetBlendWeight _blendweight (Variables|Default|GetMobilePostProcess)))
      (:Finished
        (if (Utilities|Enum|Equal(Enum) _direction "Backward")
          (Components|DestroyComponent (Variables|Default|GetMobilePostProcess)))))))

(event Custom|EndRotationTick
  ;[1 1 1] Special Tick for Replicating the Control rotation of the player to other clients. Used By HighTower Tomato Repulsor Cannon
  (Utilities|Time|ClearandInvalidateTimerbyHandle (stale Custom|StartAImRotationTick _returnvalue))
  (Utilities|FlowControl|Gate true
    (:Exit
      (Networking|ForceNetUpdate)
      (CallFunction|ReplicateAimRotation (Pawn|GetControlRotation)))))

(event Custom|TeleportInNSeconds
  (params (SecondsTillTeleport :type Float%20%28double-precision%29) (DestLocation :type Vector) (DestRotation :type Rotator))
  ;[1 1 1] Using this for special events. Need to be able to hand off some state to the players here
  (Utilities|FlowControl|RetriggerableDelay SecondsTillTeleport))

(event Custom|RemoveStormFX
  ;[1 1 1] Comment
  ;;[1 1 1] Events for managing storm-related fx, most things moved to GC_Athena_OutsideSafeZone
  ;;;[1 0 0] Mobile Post Process STRIP if ConstructScriptWorks
  (if (and (Fortnite|IsMobilePlatform) (Pawn|IsLocallyControlled))
    ;;;[1 0 0] Mobile Only Post Settings  STRIP if ConstructScriptWorks
    (Timeline|Timeline_0 Reverse)))

(event Custom|WhiteoutAndTeleport
  (params (DestLocation :type Vector) (DestRotation :type Rotator) (Instant :type Boolean) (Gravity :type Float%20%28double-precision%29 :default 1.0))
  (bind _self self)
  (bind _returnvalue (Ability|GetAbilitySystemComponent _self))
  ;[1 1 1] Using this for special events. Need to be able to hand off some state to the players here
  (CallFunction|ClientRunSnowGC Instant)
  (if (not Instant)
    (Utilities|FlowControl|Delay 3.0))
  (Variables|/Script/Engine.MovementComponent|SetVelocity "0, 0, 0" (Variables|Character|GetCharacterMovement))
  (Transformation|Teleport DestLocation DestRotation)
  (Utilities|FlowControl|Sequence
    (:then_0
      (if (== Gravity 0.1)
        (GameplayEffects|ApplyGameplayEffectToSelf :self _returnvalue :GameplayEffectClass "/Game/Athena/Prototype/Blueprints/White/GE_Lowgrav.GE_Lowgrav_C" :Level 0.0)
        (else
          (GameplayEffects|RemoveActiveGameplayEffectbySourceEffect _returnvalue "/Game/Athena/Prototype/Blueprints/White/GE_AntiGrav.GE_AntiGrav_C" 0 -1))))
    (:then_1
      (if (== Gravity 1.0)
        (Utilities|FlowControl|Delay 3.0)
        (Parachute|ForceOpenParachute false)
        (Parachute|ForceLockParachuteOpen true)))))

(event Custom|OnCharacterCustomizationCompleted
  (params (Pawn :type Fort%20Player%20Pawn%20Object%20Reference))
  ;[1 1 1] On Cosmetics Set - FX will fail if called too early because Skeletal Mesh array will be null
  (if (Variables|LobbyDisplay|GetbQueueRezIn)
    (Variables|LobbyDisplay|SetbQueueRezIn false)
    (AthenaPawn|OnResInDuringWarmup)
    (elif (Variables|LobbyDisplay|GetbQueueRezOut)
      (Variables|LobbyDisplay|SetbQueueRezOut false)
      (GameplayCue|ExecuteGameplayCue%28Burst%29Local self (Variables|Elimination|GetElimination%20Rez%20Sequence%20GC) (Ability|GameplayCue|MakeGameplayCueParameters 0.0)))))

(event Custom|ScalePlayerUp
  (Transformation|SetWorldScale3D (Variables|Character|GetMesh) (* (Variables|/Script/Engine.SceneComponent|GetRelativeScale3D (Variables|Character|GetMesh)) (Variables|Default|GetScaleRate))))

(event EventOnCharacterPartsReinitialized
  ;[1 1 1] Rimlight
  (|CallParentFunction)
  (DataDrivenCVar|BindEventtoOnDataDrivenCVarDelegate (EngineSubsystems|GetDataDrivenCVarEngineSubsystem) (EventDispatchers|CreateEvent :!event Rimlight%20Setup)) ; Handle Cvar changes
  (Utilities|FlowControl|RetriggerableDelay 0.5)
  (RimLight|RimlightSetup "")
  ;[0.4853 1 0.6917] If we are in a UGC experience, then setup a binding to the rim light intensity change
  (RimLight|UGCBindRimLightIntensityChange))

(event Custom|ClientRunSnowGC
  (params (Instant :type Boolean))
  (bind _returnvalue (Math|Conversions|ToFloat(Boolean) Instant))
  (bind _returnvalue_2 (Ability|GameplayCue|MakeGameplayCueParameters _returnvalue))
  (bind _self self)
  ;[1 1 1] Using this for special events. Need to be able to hand off some state to the players here
  (Projectile|AddGameplayCue%28Looping%29Local _self (Variables|Default|GetWhiteoutCue) _returnvalue_2)
  (Utilities|FlowControl|Delay (select Instant 0.15 3.0))
  (Default|CallOnFadetoWhiteTeleported)
  (Projectile|RemoveGameplayCue%28Looping%29Local _self (Variables|Default|GetWhiteoutCue) _returnvalue_2))

(event Custom|ScalePlayerDown
  (Transformation|SetWorldScale3D (Variables|Character|GetMesh) (/ (Variables|/Script/Engine.SceneComponent|GetRelativeScale3D (Variables|Character|GetMesh)) (Variables|Default|GetScaleRate))))

(event Creative|EventOnCreativeStartGhost
  ;[1 1 1] Creative Mode - Ghost Mode FX
  (Networking|CanExecuteCosmeticEvents
    (:True
      (EventDispatchers|BindEventtoOnPostResIn (delegate Custom|Event%20Materials%20Refreshed))
      (Projectile|AddGameplayCue%28Looping%29Local self (Variables|Default|GetPhoneGhostCue) (Ability|GameplayCue|MakeGameplayCueParameters :NormalizedMagnitude 0.0 :RawMagnitude 0.0 :Location "0,0,0" :Normal "0,0,0" :Instigator self :GameplayEffectLevel 1 :AbilityLevel 1 :bReplicateLocationWhenUsingMinimalRepProxy false)))))

(event AthenaPawn|EventOnResInDuringWarmup
  ;[1 1 1] Res in, used in the lobby
  (Rendering|SetActorHiddenInGame false)
  (GameplayCue|ExecuteGameplayCue%28Burst%29Local self (select (Reactive|Frontend|IsCosmeticObjectDisplayedinFrontendContext self) (Variables|Default|GetTeleport%20In%20GC_Frontend) (Variables|Default|GetTeleport%20In%20GC)) (Ability|GameplayCue|MakeGameplayCueParameters 0.0)))

(event EventOnTetherStarted
  (params (LastTetherPawn :type Pawn%20Object%20Reference))
  ;[1 1 1] Tether quest credit
  (bind _returnvalue (GameplayTags|GetOwnedGameplayTags (CallFunction|GetTetherPawn) :!class /Script/GameplayTags.GameplayTagAssetInterface))
  (Quest|SendComplexCustomStatEvent (Quests|GetQuestManager (Fortnite|GetFortPlayerControllerfromActor self) "Invalid") 0 (Variables|Default|GetTetherQuestTag) _returnvalue 1))

(event Collision|EventHit
  (params (MyComp :type Primitive%20Component%20Object%20Reference) (Other :type Actor%20Object%20Reference) (OtherComp :type Primitive%20Component%20Object%20Reference) (bSelfMoved :type Boolean) (HitLocation :type Vector :default "0, 0, 0") (HitNormal :type Vector :default "0, 0, 0") (NormalImpulse :type Vector :default "0, 0, 0") (Hit :type Hit%20Result%20Structure%20%28by%20ref%29)))

(event Custom|Event%20Materials%20Refreshed
  (params (PlayerPawn :type Fort%20Player%20Pawn%20Object%20Reference))
  ;[1 1 1] Creative Mode - Ghost Mode FX
  ;; Start Re-apply Highlight Timer
  (Utilities|Time|ClearandInvalidateTimerbyHandle (Variables|PawnHighlight|GetRe-apply%20Highlight%20Timer))
  (bind _returnvalue (Utilities|Time|SetTimerforNextTickbyEvent (delegate Custom|Re-apply%20Highlight)))
  (Variables|PawnHighlight|SetRe-apply%20Highlight%20Timer _returnvalue))

(event InterfacePlayerPawn|EventMeleeSwingLeft
  (params (First%20Left :type Boolean))
  ;[1 1 1] Melee swing start and end
  (if First%20Left
    (Default|CallSwingLeft)
    (else
      (Default|CallSwingLeft2))))

(event Creative|EventOnCreativeStopGhost
  ;[1 1 1] Creative Mode - Ghost Mode FX
  (Networking|CanExecuteCosmeticEvents
    (:True
      (EventDispatchers|UnbindEventfromOnPostResIn (delegate Custom|Event%20Materials%20Refreshed))
      (Projectile|RemoveGameplayCue%28Looping%29Local self (Variables|Default|GetPhoneGhostCue) (Ability|GameplayCue|MakeGameplayCueParameters :NormalizedMagnitude 0.0 :RawMagnitude 0.0 :Location "0,0,0" :Normal "0,0,0" :Instigator self :GameplayEffectLevel 1 :AbilityLevel 1 :bReplicateLocationWhenUsingMinimalRepProxy false)))))

(event EventOnTetherEnded
  (params (LastTetherPawn :type Pawn%20Object%20Reference)))

(event InterfacePlayerPawn|EventMeleeSwingRight
  (params (First%20Right :type Boolean))
  ;[1 1 1] Melee swing start and end
  (if First%20Right
    (Default|CallSwingRight)
    (else
      (Default|CallSwingRight2))))

(event InterfacePlayerPawn|EventMeleeSwingLeft_End
  ;[1 1 1] Melee swing start and end
  (Default|CallSwingLeftEnd :!class /Game/Athena/PlayerPawn_Athena.PlayerPawn_Athena_C))

(event InterfacePlayerPawn|EventMeleeSwingRight_End
  ;[1 1 1] Melee swing start and end
  (Default|CallSwingRightEnd :!class /Game/Athena/PlayerPawn_Athena.PlayerPawn_Athena_C))

(event Creative|EventOnCreativeModeratorModeActivate
  ;[1 1 1] Creative Mode - Moderator Mode FX
  (Projectile|AddGameplayCue%28Looping%29Local self (Variables|Default|GetPhoneModeratorModeGameplayCue) (Ability|GameplayCue|MakeGameplayCueParameters :NormalizedMagnitude 0.0 :RawMagnitude 0.0 :Location "0,0,0" :Normal "0,0,0" :Instigator self :GameplayEffectLevel 1 :AbilityLevel 1 :bReplicateLocationWhenUsingMinimalRepProxy false)))

(event AthenaPawn|EventOnResInDuringStyleTransference
  ;[1 1 1] Handle Style Transfer Effect, used in Lobby
  (Rendering|SetActorHiddenInGame true)
  (GameplayCue|ExecuteGameplayCue%28Burst%29Local self (Variables|Default|GetStyle%20Transfer%20GC_Frontend) (Ability|GameplayCue|MakeGameplayCueParameters 0.0)))

(event Creative|EventOnCreativeModeratorModeDeactivate
  ;[1 1 1] Creative Mode - Moderator Mode FX
  (Projectile|RemoveGameplayCue%28Looping%29Local self (Variables|Default|GetPhoneModeratorModeGameplayCue) (Ability|GameplayCue|MakeGameplayCueParameters :NormalizedMagnitude 0.0 :RawMagnitude 0.0 :Location "0,0,0" :Normal "0,0,0" :Instigator self :GameplayEffectLevel 1 :AbilityLevel 1 :bReplicateLocationWhenUsingMinimalRepProxy false)))

(event Replay|EventOnEnteredReplayRelevancy
  ;[1 1 1] Entering/Leaving Relevancy in Replays
  (Effects|SpawnEmitterAttached "/Game/Spectating/Effects/P_Player_Relevance_Res.P_Player_Relevance_Res" (Variables|Character|GetCapsuleComponent)))

(event MeleeSwingAndAnimTrailEventsPlayerInterface|EventBindToAnimTrailEvents
  (params (AnimNotifyBegin :type Delegate%20%28by%20ref%29) (AnimNotifyEnd :type Delegate%20%28by%20ref%29))
  ;[1 1 1] MeleeSwingAndAnimTrailEventsPlayerInterface
  (Default|BindEventtoAnimNotifyBegin AnimNotifyBegin :!class /Game/Athena/PlayerPawn_Athena_Generic.PlayerPawn_Athena_Generic_C)
  (Default|BindEventtoAnimNotifyEnd AnimNotifyEnd :!class /Game/Athena/PlayerPawn_Athena_Generic.PlayerPawn_Athena_Generic_C))

(event MeleeSwingAndAnimTrailEventsPlayerInterface|EventBindToSwingFX
  (params (SwingLeft :type Delegate%20%28by%20ref%29) (SwingLeftEnd :type Delegate%20%28by%20ref%29) (SwingRight :type Delegate%20%28by%20ref%29) (SwingRightEnd :type Delegate%20%28by%20ref%29))
  ;[1 1 1] MeleeSwingAndAnimTrailEventsPlayerInterface
  (Default|BindEventtoSwingLeft SwingLeft :!class /Game/Athena/PlayerPawn_Athena.PlayerPawn_Athena_C)
  (Default|BindEventtoSwingLeftEnd SwingLeftEnd :!class /Game/Athena/PlayerPawn_Athena.PlayerPawn_Athena_C)
  (Default|BindEventtoSwingRight SwingRight :!class /Game/Athena/PlayerPawn_Athena.PlayerPawn_Athena_C)
  (Default|BindEventtoSwingRightEnd SwingRightEnd :!class /Game/Athena/PlayerPawn_Athena.PlayerPawn_Athena_C))

(event MeleeSwingAndAnimTrailEventsPlayerInterface|EventBindToSwingFX2
  (params (SwingLeft2 :type Delegate%20%28by%20ref%29) (SwingRight2 :type Delegate%20%28by%20ref%29))
  ;[1 1 1] MeleeSwingAndAnimTrailEventsPlayerInterface
  (Default|BindEventtoSwingLeft2 SwingLeft2 :!class /Game/Athena/PlayerPawn_Athena.PlayerPawn_Athena_C)
  (Default|BindEventtoSwingRight2 SwingRight2 :!class /Game/Athena/PlayerPawn_Athena.PlayerPawn_Athena_C))

(event StormSurge|EventOnReceivedStormSurgeDamage
  (bind _capsulecomponent (Variables|Character|GetCapsuleComponent))
  (Effects|SpawnEmitterAttached "/Game/Effects/Fort_Effects/Effects/Storm/P_Storm_Cap.P_Storm_Cap" _capsulecomponent)
  (Audio|SpawnSoundAttached "/Game/Athena/Sounds/LTMs/StormCap/StormCap_Lightning_Strike_Cue.StormCap_Lightning_Strike_Cue" _capsulecomponent))

(event Custom|WaitingForMaterials
  ;[1 1 1] Creative Move - Invulnerable FX
  (if (or (not (Utilities|IsValid (Variables|Pawn|GetPlayerState) :!class /Script/Engine.KismetSystemLibrary)) (Utilities|Array|IsEmpty (Variables|CharacterMIDs|GetPawnMaterials_ALL)))
    (bind _returnvalue (Utilities|Time|SetTimerbyEvent (delegate Custom|WaitingForMaterials) 0.5)) ; When player spawns with invulnerability on server, it will ask to show this fx but the character might not be\n fully loaded (ie FinishedCharacterCustomization is not called yet or no player state yet)
    (Variables|PawnHighlight|SetWaitingForMaterialsTimer _returnvalue)
    (else
      (Projectile|AddGameplayCue%28Looping%29Local self (Variables|Default|GetPhoneInvulnerableGameplayCue) (Ability|GameplayCue|MakeGameplayCueParameters :NormalizedMagnitude 0.0 :RawMagnitude 0.0 :Location "0,0,0" :Normal "0,0,0" :Instigator self :GameplayEffectLevel 1 :AbilityLevel 1 :bReplicateLocationWhenUsingMinimalRepProxy false)))))

(event Creative|EventOnCreativeStartInvulnerable
  ;[1 1 1] Creative Move - Invulnerable FX
  (Networking|CanExecuteCosmeticEvents
    (:True
      (if (or (not (Utilities|IsValid (Variables|Pawn|GetPlayerState) :!class /Script/Engine.KismetSystemLibrary)) (Utilities|Array|IsEmpty (Variables|CharacterMIDs|GetPawnMaterials_ALL)))
        (bind _returnvalue (Utilities|Time|SetTimerbyEvent (delegate Custom|WaitingForMaterials) 0.5)) ; When player spawns with invulnerability on server, it will ask to show this fx but the character might not be\n fully loaded (ie FinishedCharacterCustomization is not called yet or no player state yet)
        (Variables|PawnHighlight|SetWaitingForMaterialsTimer _returnvalue)
        (else
          (Projectile|AddGameplayCue%28Looping%29Local self (Variables|Default|GetPhoneInvulnerableGameplayCue) (Ability|GameplayCue|MakeGameplayCueParameters :NormalizedMagnitude 0.0 :RawMagnitude 0.0 :Location "0,0,0" :Normal "0,0,0" :Instigator self :GameplayEffectLevel 1 :AbilityLevel 1 :bReplicateLocationWhenUsingMinimalRepProxy false)))))))

(event MeleeSwingAndAnimTrailEventsPlayerInterface|EventUnbindFromSwingFX2
  (params (SwingLeft2 :type Delegate%20%28by%20ref%29) (SwingRight2 :type Delegate%20%28by%20ref%29))
  ;[1 1 1] MeleeSwingAndAnimTrailEventsPlayerInterface
  (Default|UnbindEventfromSwingLeft2 SwingLeft2 :!class /Game/Athena/PlayerPawn_Athena.PlayerPawn_Athena_C)
  (Default|UnbindEventfromSwingRight2 SwingRight2 :!class /Game/Athena/PlayerPawn_Athena.PlayerPawn_Athena_C))

(event Replay|EventOnLeftReplayRelevancy
  ;[1 1 1] Entering/Leaving Relevancy in Replays
  (Effects|SpawnEmitteratLocation "/Game/Spectating/Effects/P_Player_Relevance_Res.P_Player_Relevance_Res" (Transformation|GetWorldLocation (Variables|Character|GetCapsuleComponent)) (Transformation|GetWorldRotation (Variables|Character|GetCapsuleComponent))))

(event MeleeSwingAndAnimTrailEventsPlayerInterface|EventUnbindFromSwingFX
  (params (SwingLeft :type Delegate%20%28by%20ref%29) (SwingLeftEnd :type Delegate%20%28by%20ref%29) (SwingRight :type Delegate%20%28by%20ref%29) (SwingRightEnd :type Delegate%20%28by%20ref%29))
  ;[1 1 1] MeleeSwingAndAnimTrailEventsPlayerInterface
  (Default|UnbindEventfromSwingLeft SwingLeft :!class /Game/Athena/PlayerPawn_Athena.PlayerPawn_Athena_C)
  (Default|UnbindEventfromSwingLeftEnd SwingLeftEnd :!class /Game/Athena/PlayerPawn_Athena.PlayerPawn_Athena_C)
  (Default|UnbindEventfromSwingRight SwingRight :!class /Game/Athena/PlayerPawn_Athena.PlayerPawn_Athena_C)
  (Default|UnbindEventfromSwingRightEnd SwingRightEnd :!class /Game/Athena/PlayerPawn_Athena.PlayerPawn_Athena_C))

(event MeleeSwingAndAnimTrailEventsPlayerInterface|EventUnbindFromAnimTrailEvents
  (params (AnimNotifyBegin :type Delegate%20%28by%20ref%29) (AnimNotifyEnd :type Delegate%20%28by%20ref%29))
  ;[1 1 1] MeleeSwingAndAnimTrailEventsPlayerInterface
  (Default|UnbindEventfromAnimNotifyBegin AnimNotifyBegin :!class /Game/Athena/PlayerPawn_Athena_Generic.PlayerPawn_Athena_Generic_C)
  (Default|UnbindEventfromAnimNotifyEnd AnimNotifyEnd :!class /Game/Athena/PlayerPawn_Athena_Generic.PlayerPawn_Athena_Generic_C))

(event Creative|EventOnCreativeStopInvulnerable
  ;[1 1 1] Creative Move - Invulnerable FX
  (Utilities|Time|ClearandInvalidateTimerbyHandle (Variables|PawnHighlight|GetWaitingForMaterialsTimer))
  (Projectile|RemoveGameplayCue%28Looping%29Local self (Variables|Default|GetPhoneInvulnerableGameplayCue) (Ability|GameplayCue|MakeGameplayCueParameters :NormalizedMagnitude 0.0 :RawMagnitude 0.0 :Location "0,0,0" :Normal "0,0,0" :Instigator self :GameplayEffectLevel 1 :AbilityLevel 1 :bReplicateLocationWhenUsingMinimalRepProxy false)))

(event Custom|Play%20Respawn%20FX
  ;[1 1 1] Creative Respawn FX
  (Networking|CanExecuteCosmeticEvents
    (:True
      (if (> (Utilities|Array|Length (Variables|CharacterMIDs|GetPawnMaterials_ALL)) 0)
        (GameplayCue|ExecuteGameplayCue%28Burst%29Local self (Variables|Default|GetCreative%20Respawn%20Teleportation%20GC) (Ability|GameplayCue|MakeGameplayCueParameters 0.0))
        (CallFunction|ActivateInvincibleVisuals (Variables|FortPawn|GetSpawnImmunityTime))
        (else
          ;;[1 1 1] Need to wait till materials are ready
          (Utilities|Time|SetTimerbyEvent (delegate Custom|Play%20Respawn%20FX) 0.1))))))

(event Spawn|EventOnSpawnImmunityTimeReset
  ;[1 1 1] Handle event where spawn time gets reset to zero
  (PawnHighlight|RemovePawnHighlightFromSource self))

(event PawnHighlight|EventAddHighlightFromSource
  (params (Source :type Object%20Reference) (Highlight :type Pawn%20Highlight%20Structure) (OnlyUpdateIfCurrent :type Boolean))
  (PawnHighlight|HighlightPawnFromSource Source Highlight OnlyUpdateIfCurrent))

(event PawnHighlight|EventRemoveHighlightFromSource
  (params (Source :type Object%20Reference))
  (PawnHighlight|RemovePawnHighlightFromSource Source))

(event EventOnDisplayHitNotify
  (params (DamageDealt :type Float%20%28single-precision%29) (bCriticalHit :type Boolean) (bFatalHit :type Boolean) (bShieldHit :type Boolean) (HitActor :type Actor%20Object%20Reference) (HitTags :type Gameplay%20Tag%20Container%20Structure) (HitLocation :type Vector :default "0, 0, 0"))
  (EventDispatchers|CallDisplayHitNotify DamageDealt bCriticalHit bFatalHit bShieldHit HitActor HitLocation HitTags))

(event Custom|Hide%20Character%20Parts
  (params (isHidden :type Boolean))
  (Variables|Default|SetIs%20Hidding%20CharacterParrts isHidden)
  (Cosmetics|WhenCosmeticPartApplied (delegate Custom|OnPart)))

(event Custom|OnPart
  (params (Pawn :type Fort%20Player%20Pawn%20Object%20Reference) (Part :type Custom%20Character%20Part%20Object%20Reference) (MeshPart :type Skeletal%20Mesh%20Component%20Object%20Reference) (PartType :type EFortCustomPartType%20Enum :default "Head"))
  (bind _is_hidding_characterparrts (Variables|Default|GetIs%20Hidding%20CharacterParrts))
  (Utilities|IsValid MeshPart
    (:Is%20Valid
      (Development|SetHiddeninGame MeshPart _is_hidding_characterparrts false)
      (for _array_element (Components|GetChildrenComponents MeshPart true)
        (if (Components|ComponentHasTag _array_element "LockVisibility")
          _
          (elif (Utilities|IsValid (Components|GetOwner _array_element) :!class /Script/Engine.KismetSystemLibrary)
            (Rendering|SetActorHiddenInGame (Components|GetOwner _array_element) false)))
        (Development|SetHiddeninGame _array_element _is_hidding_characterparrts false)))))

(event Custom|OnRep_PlayRespawnFXOnSpawn
  ;[1 1 1] Due to small delay when spawning Pawn from server, the player state can be missing for a few frame, so wait until its there, it impacts logic below
  (Utilities|IsValid (Variables|Pawn|GetPlayerState) :!label _label
    (:Is%20Valid
      (if (Variables|Respawn|GetPlayRespawnFXOnSpawn)
        (CallFunction|PlayRespawnFX)))
    (:Is%20Not%20Valid
      (Utilities|FlowControl|Delay 0.2)
      (goto _label))))

(event BPI_PlayerPawnAthena|EventCallRezInEvent
  (Default|CallResIn))

(fn ApplyPawnHighlight
  (params (Source :type Object%20Reference) (HitGlow :type Pawn%20Highlight%20Structure))
  (bind _current_highlight (Variables|PawnHighlight|GetCurrent%20Highlight))
  (bind (_priority_28_e2e1b5344846e187b9c11b863a7f0698 _inner_21_4cc2801147ea190de16f59b34f36853e _outer_22_5a1d7d0543d303e8b54b66a7f7bd2e2e _fresnelbrightness_23_52b0f96447ff640f47df2895b0602e92 _fresnelexponent_24_b427cf0c441aa37ed49833bf7579de6d _usespulse_25_e29229f64e540f0617e4c4987ad77605) (Utilities|Struct|BreakPawnHighlight _current_highlight))
  (Variables|PawnHighlight|SetCurrent%20Highlight%20Source Source)
  (Variables|PawnHighlight|SetCurrent%20Highlight HitGlow)
  (for _array_element (Variables|CharacterMIDs|GetPawnMaterials_ALL)
    (Rendering|Material|SetVectorParameterValue _array_element "HitGlow" _outer_22_5a1d7d0543d303e8b54b66a7f7bd2e2e :!class /Script/Engine.MaterialInstanceDynamic)
    (Rendering|Material|SetVectorParameterValue _array_element "HitGlowOuter" _inner_21_4cc2801147ea190de16f59b34f36853e :!class /Script/Engine.MaterialInstanceDynamic)
    (Rendering|Material|SetScalarParameterValue _array_element "HitGlowFresnelBrightness" (+ _fresnelbrightness_23_52b0f96447ff640f47df2895b0602e92 (Variables|Default|GetHitGlowBrightnessScalar)) :!class /Script/Engine.MaterialInstanceDynamic)
    (Rendering|Material|SetScalarParameterValue _array_element "HitGlowFresnelExponent" _fresnelexponent_24_b427cf0c441aa37ed49833bf7579de6d :!class /Script/Engine.MaterialInstanceDynamic)
    (Rendering|Material|SetScalarParameterValue _array_element "HitGlowUsesPulse" _usespulse_25_e29229f64e540f0617e4c4987ad77605 :!class /Script/Engine.MaterialInstanceDynamic)))

(fn AttemptToSetRimLight
  ;[1 1 1] Ready to setup player rim light?
  (if (Utilities|Array|IsValidIndex (Variables|CharacterMIDs|GetPawnMaterials_ALL))
    (RimLight|RimlightSetup "")
    (PhysicalRimLight|ApplyPhysicalRimLightLocationandColor (Variables|FortPlayerPawnAthena|GetRimLightComponent) (Variables|RimLight|GetRim%20Light%20Intensity))
    (Utilities|Time|ClearTimerbyFunctionName 0 "AttemptToSetRimLight")
    ;[0.4354 0.506 1] Setup a binding to CVar changes such as graphic quality presets (Epic / High / Medium / Low)
    (DataDrivenCVar|BindEventtoOnDataDrivenCVarDelegate (EngineSubsystems|GetDataDrivenCVarEngineSubsystem) (EventDispatchers|CreateEvent :!event Rimlight%20Setup)) ; Handle Cvar changes
    ;[0.4853 1 0.6917] If we are in a UGC experience, then setup a binding to the rim light intensity change
    (RimLight|UGCBindRimLightIntensityChange)))

(fn EnableWaterAudio
  (params (Is%20Entering%20Water :type Boolean))
  (if Is%20Entering%20Water
    (bind _returnvalue (Audio|SpawnSoundatLocation :Sound (select (Pawn|IsLocallyViewed) (Variables|AthenaAudio|GetSFX_WaterEntry_1P) (Variables|AthenaAudio|GetSFX_WaterEntry_3P)) :Location (+ (Transformation|GetActorLocation) "0, 0,50.000000") :VolumeMultiplier (Math|Float|SelectFloat 0.6 1.0 (Pawn|IsLocallyViewed)) :PitchMultiplier 1.0 :StartTime 0.0 :bAutoDestroy true))
    (Audio|Parameter|SetFloatParameter _returnvalue "EntrySpeed" (Math|Float|MapRangeClamped (Math|Vector|VectorLength (Transformation|GetVelocity)) 0.0 1.2e+03 0.0 1.0) :!class /Script/AudioExtensions.AudioParameterControllerInterface)
    (elif (Variables|Gameplay|GetbIsWaterJump)
      (Audio|PlaySoundatLocation (select (Pawn|IsLocallyViewed) (Variables|AthenaAudio|GetSFX_WaterExit_1P) (Variables|AthenaAudio|GetSFX_WaterExit_3P)) (+ (Transformation|GetActorLocation) "0, 0,50.000000")))))

(fn GalileoMaterialSetup
  (for _array_element (Variables|CharacterMIDs|GetPawnMaterials_ALL)
    (Rendering|Material|SetVectorParameterValue _array_element "Blue" "(R=0.082682,G=1.168436,B=1.500000,A=1.000000)" :!class /Script/Engine.MaterialInstanceDynamic)
    (Rendering|Material|SetVectorParameterValue _array_element "DeRezEdgeGlow" "(R=0.250126,G=4.701368,B=5.000000,A=1.000000)" :!class /Script/Engine.MaterialInstanceDynamic)
    (Rendering|Material|SetVectorParameterValue _array_element "GridColor" "(R=0.000000,G=0.821185,B=1.000000,A=1.000000)" :!class /Script/Engine.MaterialInstanceDynamic)))

(fn GetHighestPriorityHighlight
  (bind _pawnhighlights (Variables|PawnHighlight|GetPawnHighlights))
  (bind _keys (Utilities|Map|Keys _pawnhighlights))
  (for _array_element _keys
    (if (>= (Utilities|Struct|BreakPawnHighlight (Utilities|Map|Find _pawnhighlights _array_element)) (Utilities|Struct|BreakPawnHighlight (Utilities|Map|Find (Variables|PawnHighlight|GetPawnHighlights) (LocalVariables|GetBest%20Source))))
      (LocalVariables|SetBest%20Source _array_element) ; Defaults to no highlight with priority -1. \nThis will return no highlight if there is nothing in the array.
    ))
  (return (LocalVariables|GetBest%20Source) (Utilities|Map|Find (Variables|PawnHighlight|GetPawnHighlights) (LocalVariables|GetBest%20Source))))

(fn GetAimPoint
  (return (Pawn|GetControlRotation)))

(fn GetCurrentWeaponMeshComponents
  (return (Variables|Default|GetCurrentWeaponMeshes)))

(fn GetEliminationRezGC
  (Utilities|FlowControl|Sequence
    (:then_0
      ;[0.1287 0.05284 0.15 0.5] Attempt to use custom GC from source weapon
      (bind _asfort_weapon (Utilities|Casting|CastToFortWeapon (Ability|EffectContext|GetSourceObject (Variables|Elimination|GetOn%20Death%20Effect%20Context))))
      (Utilities|IsValid (Data|GetWeaponData _asfort_weapon)
        (:Is%20Valid
          (bind _itemdata (Class|ItemDefinitionBase|GetItemData (Data|GetWeaponData _asfort_weapon))
            (:Found
              (return (Utilities|Struct|BreakRezOutOverrideItemData _itemdata)))))))
    (:then_1
      ;[0.4635 0 0 0.5] Setup to give Tactical Elim by Default and non-Tac Elim by Choice
      (return (select (GameplayTags|HasMatchingGameplayTag self "(TagName=\"Creative.EliminationManager.TeleportBotFX\")") (Variables|Elimination|GetTeleport%20Bot%20Elimination%20Rez%20Sequence%20GC) (Variables|Elimination|GetElimination%20Rez%20Sequence%20GC))))))

(fn GetFallingAudioLoopComponent
  (return (Variables|Default|GetFallingAudioLoop_Inst)))

(fn GetForwardAxis
  (return (.x (Math|Conversions|ToVector(Vector2D) (Class|FortMovementBlueprintLibrary|GetLocalMoveInputVector self false false)))))

(fn GetLocationInCircle_XY
  (return (Math|Vector|Lerp(Vector) (Math|Vector|MakeVector 0.0 0.0 (Variables|VFX|GetLightningFlashHeight)) (Math|Vector|MakeVector (* (Math|Trig|Sin(Radians) (Math|Float|Lerp 0.0 (* (Math|Trig|GetPI) 2) (Math|Random|RandomFloat))) (Variables|VFX|GetLightningFlashDiameter)) (* (Math|Trig|Cos(Radians) (Math|Float|Lerp 0.0 (* (Math|Trig|GetPI) 2) (Math|Random|RandomFloat))) (Variables|VFX|GetLightningFlashDiameter)) (Variables|VFX|GetLightningFlashHeight)) (Math|Random|RandomFloatinRange 0.25 1))))

(fn GetOnDeathController
  (return (Variables|Elimination|GetOn%20Death%20Controller)))

(fn GetOverrideSearchMontage
  (return (Variables|InteractSearch|GetOverrideSearchMontageInfo)))

(fn GetPawnIsInLocalSquad
  (bind _returnvalue (Fortnite|Utilities|AttemptGetPlayerStateforPawn self))
  (bind _asfort_player_state_athena (Utilities|Casting|CastToFortPlayerStateAthena _returnvalue)
    (:then
      (LocalVariables|SetThisPawnSquadID (Class|FortPlayerStateAthena|GetSquadId _asfort_player_state_athena))
      (bind _asfort_player_pawn (Utilities|Casting|CastToFortPlayerPawn (Game|GetPlayerPawn 0))
        (:then
          (bind _returnvalue_2 (Fortnite|Utilities|AttemptGetPlayerStateforPawn _asfort_player_pawn))
          (bind _asfort_player_state_athena_2 (Utilities|Casting|CastToFortPlayerStateAthena _returnvalue_2)
            (:then
              (return (== (LocalVariables|GetThisPawnSquadID) (Class|FortPlayerStateAthena|GetSquadId _asfort_player_state_athena_2))))
            (:CastFailed
              _)))
        (:CastFailed
          _))
      ;[1 1 1] Check for veiwing target if spectator is part of squad.
      (bind _asfort_player_state_athena_3 (Utilities|Casting|CastToFortPlayerStateAthena (Variables|/Script/Engine.Controller|GetPlayerState (Fortnite|GetFirstLocalFortPlayerController true true)))
        (:then
          (return (== (LocalVariables|GetThisPawnSquadID) (Class|FortPlayerStateAthena|GetSquadId _asfort_player_state_athena_3))))
        (:CastFailed
          (return false))))
    (:CastFailed
      (return false))))

(fn GetRightAxis
  (return (.y (Math|Conversions|ToVector(Vector2D) (Class|FortMovementBlueprintLibrary|GetLocalMoveInputVector self false false)))))

(fn GetStormAudioLoopComponent
  (return (Variables|Default|GetStormAudioLoop_Inst)))

(fn GetStormAudioVolume
  (return (Variables|AthenaAudio|GetStormAudioVolume)))

(fn GetStormAuraParticlesComponent
  (return (Variables|VFX|GetStormAuraParticles)))

(fn handleCvarsUpdated
  (params (CVarName :type String)))

(fn HighlightPawnfromSource
  (params (Source :type Object%20Reference) (Highlight :type Pawn%20Highlight%20Structure) (Only%20Update%20If%20Current :type Boolean))
  (bind (_value _returnvalue) (Utilities|Map|Find (Variables|PawnHighlight|GetPawnHighlights) (Variables|GetSource)))
  (if (not Only%20Update%20If%20Current)
    (if (not _returnvalue)
      (Utilities|Map|Add (Variables|PawnHighlight|GetPawnHighlights) (Variables|GetSource) (Variables|GetHighlight)))
    (if (>= (Utilities|Struct|BreakPawnHighlight (Variables|GetHighlight)) (Utilities|Struct|BreakPawnHighlight (Variables|PawnHighlight|GetCurrent%20Highlight)))
      _)
    (elif (Utilities|Equal(Object) (Variables|GetSource) (Variables|PawnHighlight|GetCurrent%20Highlight%20Source))
      _))
  (CallFunction|ApplyPawnHighlight (Variables|GetSource) (Variables|GetHighlight)))

(fn InWarmup
  (bind _returnvalue (BattleRoyaleGamePhaseLogic|GetBattleRoyaleGamePhaseLogicComponent))
  (Utilities|IsValid _returnvalue
    (:Is%20Valid
      (if (or (Utilities|Enum|Equal(Enum) (BattleRoyaleGamePhaseLogic|GetGamePhase _returnvalue) "Warmup") (Utilities|Enum|Equal(Enum) (BattleRoyaleGamePhaseLogic|GetGamePhase _returnvalue) "Setup"))
        (Variables|Default|SetbWasEverInWarmUp true)
        (return true)
        (else
          (return false))))
    (:Is%20Not%20Valid
      (return false))))

(fn IsInWarmup
  (bind _binwarmup (CallFunction|InWarmup))
  (return _binwarmup))

(fn IsSwingAttached
  (bind _returnvalue (Pawn|GetController))
  (Utilities|IsValid _returnvalue
    (:Is%20Valid
      (return (and (Utilities|IsValid (Variables|Swinging|GetSwinging_AttachBGA) :!class /Script/Engine.KismetSystemLibrary) (Swinging|IsSwingAttached (Actor|GetComponentbyClass _returnvalue "/Script/FortniteGame.FortControllerComponent_Swinging")))))))

(fn IsSwingGrounded
  (bind _returnvalue (Pawn|GetController))
  (Utilities|IsValid _returnvalue
    (:Is%20Valid
      (return (Swinging|IsSwingingGrounded (Actor|GetComponentbyClass _returnvalue "/Script/FortniteGame.FortControllerComponent_Swinging"))))
    (:Is%20Not%20Valid
      (return false))))

(fn IsSwinging
  (bind _returnvalue (Pawn|GetController))
  (bind _returnvalue_2 (Actor|GetComponentbyClass _returnvalue "/Script/FortniteGame.FortControllerComponent_Swinging"))
  (Utilities|IsValid _returnvalue
    (:Is%20Valid
      (Utilities|IsValid _returnvalue_2
        (:Is%20Valid
          (return (Swinging|IsinSwingingMovement _returnvalue_2)))))))

(fn LockonSearchCompleteTimedOut
  (Variables|Default|SetbLockOnInteractionComplete false)
  (interact|StopItemInteractionAbility "TimedOut" "(TagName=\"\")"))

(fn OnLockOnSearchCompleted
  (Variables|Default|SetbLockOnInteractionComplete true)
  (Utilities|Time|SetTimerbyFunctionName 0 "LockonSearchCompleteTimedOut" 4.0)
  (return true))

(fn OnRep_PlayerAimRotation)

(fn OnStartLongInteract
  (params (ReceivingActor :type Actor%20Object%20Reference))
  (bind _self self)
  (bind _returnvalue (Ability|GetAbilitySystemComponent _self))
  (Utilities|Casting|CastToFortPlayerPawnAthena ReceivingActor
    (:then
      (if (Utilities|IsValid _returnvalue :!class /Script/Engine.KismetSystemLibrary)
        (if (GameplayTags|HasMatchingGameplayTag _returnvalue "(TagName=\"Ability.MedicPack.Equipped\")")
          (Development|PrintString "Medic is reviving!!!")))
      (return false false false "Search"))
    (:CastFailed
      (if (Actor|ActorHasTag ReceivingActor "Athena.Item.Blade")
        (Development|PrintString "Blade Interaction")
        (return false false false "LockOnSearch")
        (elif (Actor|ActorHasTag ReceivingActor (Variables|Default|GetFirefliesTag))
          ;[1 1 1] Firefly Search Montage Logic
          (Tags|AddNonReplicatedTagtoPawn (GameplayTags|MakeLiteralGameplayTag "(TagName=\"Athena.Gameplay.Status.CatchFirefly\")"))))
      (return false false false "Search"))))

(fn PawnIsDecoy
  (return (GameplayTags|HasTag (Variables|GameplayTags|GetGameplayTags) (Variables|VFX|GetTC_PawnAthenaDecoy))))

(fn PawnIsNPC
  (return (GameplayTags|HasTag (Variables|GameplayTags|GetGameplayTags) (Variables|VFX|GetTC_PawnAthenaNPC))))

(fn PlayDeathEffects
  (GameplayCue|ExecuteGameplayCue%28Burst%29Local self (CallFunction|GetEliminationRezGC) (Ability|GameplayCue|MakeGameplayCueParameters :NormalizedMagnitude 0.0 :RawMagnitude 0.0 :EffectContext (Variables|Elimination|GetOn%20Death%20Effect%20Context) :AggregatedSourceTags (Variables|Elimination|GetOn%20Death%20Damage%20Tags) :Location "0,0,0" :Normal (Variables|Elimination|GetOn%20Death%20Momentum) :GameplayEffectLevel 1 :AbilityLevel 1 :bReplicateLocationWhenUsingMinimalRepProxy false)) ; Death Visual Sequence - includes Animations, Sound, Material Changes and VFXs
)

(fn QueueTeleportIn
  (Variables|LobbyDisplay|SetbQueueRezIn true))

(fn QueueTeleportOut
  (Variables|LobbyDisplay|SetbQueueRezOut true))

(fn RemovePawnHighlightfromSource
  (params (Source :type Object%20Reference))
  (bind (_source _top_priority) (PawnHighlight|GetHighestPriorityHighlight))
  ;[1 1 1] Find the Source in the Actor/Color array and remove it.
  (Utilities|Map|Remove (Variables|PawnHighlight|GetPawnHighlights) Source)
  (if (Utilities|Equal(Object) (Variables|PawnHighlight|GetCurrent%20Highlight%20Source) (Variables|GetSource))
    (CallFunction|ApplyPawnHighlight _source _top_priority) ; If the source we removed was actively applying a color, find and reapply\nthe next highest priority one.
    (GameplayCue|ExecuteGameplayCue%28Burst%29Local self (Variables|Default|GetRemovePawnHighlightGC) (Ability|GameplayCue|MakeGameplayCueParameters 0.0))))

(fn RimlightSetup
  (params (CVarName :type String))
  (RimLight|SetupRimLight (Variables|FortPlayerPawnAthena|GetRimLightComponent) (Variables|RimLight|GetUGC%20RimLightIntensity)))

(fn SetMaterialRimlightIntensity
  (params (Multiplier :type Float%20%28double-precision%29))
  (bind _skinned_mesh_component (Utilities|ForEachSkinnedMeshComponentonPawn self)
    (:Loop%20Body
      (Rendering|Material|SetDefaultCustomPrimitiveDataFloat _skinned_mesh_component 0 Multiplier)
      (Development|PrintString (Utilities|String|Append "SET Custom Prim Data - " (Utilities|String|ToString(Float) Multiplier))))))

(fn SetPartyInProgress
  (params (bIsPartyInProgress :type Boolean))
  (if bIsPartyInProgress
    (Variables|LobbyDisplay|SetbFrontendPartyInProgress true)
    (elif (Variables|LobbyDisplay|GetbFrontendPartyInProgress)
      (Variables|LobbyDisplay|SetbFrontendPartyInProgress false))))

(fn StopLoopingAudio
  (Utilities|FlowControl|Sequence
    (:then_0
      (bind _fallingaudioloop_inst (Variables|Default|GetFallingAudioLoop_Inst)
        (:then
          (Audio|Components|Audio|FadeOut _fallingaudioloop_inst 1.0 0.0 "Linear" :!class /Script/Engine.AudioComponent))
        (:else)))
    (:then_1
      (bind _stormaudioloop_inst (Variables|Default|GetStormAudioLoop_Inst)
        (:then
          (Audio|Components|Audio|FadeOut _stormaudioloop_inst 1.0 0.0 "Linear" :!class /Script/Engine.AudioComponent))
        (:else)))))

(fn Swing_GetStartPoint
  (Utilities|IsValid (Variables|/Script/FortniteGame.FortWeapon|GetWeaponMesh (Variables|Weapon|GetCurrentWeapon))
    (:Is%20Valid
      (return (Math|Vector|SelectVector (Collision|GetComponentBounds (Variables|Character|GetMesh)) (Transformation|GetSocketLocation (Variables|Character|GetMesh) "Hand_R") (Variables|Swinging|GetSwinging_LockedToHip))))))

(fn SwingAttached
  (params (AttachLocation :type Vector) (AttachedActor :type Actor%20Object%20Reference))
  (bind _returnvalue (Pawn|GetController))
  (Utilities|IsValid _returnvalue
    (:Is%20Valid
      (Swinging|SwingAttached (Actor|GetComponentbyClass _returnvalue "/Script/FortniteGame.FortControllerComponent_Swinging") AttachedActor))))

(fn SwingDetached
  (bind _returnvalue (Pawn|GetController))
  (Utilities|IsValid _returnvalue
    (:Is%20Valid
      (Swinging|SwingDetached (Actor|GetComponentbyClass _returnvalue "/Script/FortniteGame.FortControllerComponent_Swinging")))))

(fn Swinging_AttachLocation
  (bind _swinging_attachbga (Variables|Swinging|GetSwinging_AttachBGA))
  (Utilities|IsValid _swinging_attachbga
    (:Is%20Valid
      (return (Transformation|GetActorLocation _swinging_attachbga) true))
    (:Is%20Not%20Valid
      (return (Transformation|GetActorLocation) true))))

(fn Swinging_CheckGrounded
  (bind _outhit (Collision|SphereTraceByProfile :Start (Transformation|GetWorldLocation (Variables|Character|GetCapsuleComponent)) :End (Math|Vector|vector+vector (Transformation|GetWorldLocation (Variables|Character|GetCapsuleComponent)) (Math|Vector|MakeVector 0.0 0.0 (* (Components|Capsule|GetScaledCapsuleHalfHeight (Variables|Character|GetCapsuleComponent)) -4e+01))) :Radius (Components|Capsule|GetScaledCapsuleRadius (Variables|Character|GetCapsuleComponent)) :ProfileName "FortPawnCapsule" :bTraceComplex false :DrawDebugType "None" :bIgnoreSelf true :TraceColor "(R=1.000000,G=0.000000,B=0.000000,A=1.000000)" :TraceHitColor "(R=0.000000,G=1.000000,B=0.000000,A=1.000000)" :DrawTime 7.0))
  (return (Collision|BreakHitResult _outhit) (Math|Vector|SelectVector (.Normal (Collision|BreakHitResult _outhit)) "0, 0,1.000000" (Collision|BreakHitResult _outhit)) (.Distance (Collision|BreakHitResult _outhit))))

(fn Swinging_ResetMeshTransform)

(fn Swinging_SetJumpHeight
  (params (Jump%20ZVelocity :type Float%20%28double-precision%29))
  (bind _asfort_movement_comp_character (Utilities|Casting|CastToFortMovementComp_Character (Variables|Character|GetCharacterMovement)))
  (Variables|Swinging|SetSwinging_ReelingIn true))

(fn Swinging_StartReelingIn
  (Utilities|Casting|CastToFortMovementComp_Character (Variables|Character|GetCharacterMovement))
  (Variables|Swinging|SetSwinging_ReelingIn true))

(fn Swinging_StopReelingIn
  (Utilities|Casting|CastToFortMovementComp_Character (Variables|Character|GetCharacterMovement))
  (Variables|Swinging|SetSwinging_ReelingIn false))

(fn SwingingVectorToAttachPoint
  (params (caller%20ID :type String))
  (Utilities|IsValid (Variables|Swinging|GetSwinging_AttachBGA)
    (:Is%20Valid
      (return (Utilities|SwingingVectortoAttachPoint)))
    (:Is%20Not%20Valid
      (Development|PrintString (Utilities|String|Append "ERROR: BGA invalid. ID: " caller%20ID) true true "(R=1.000000,G=0.194014,B=0.144247,A=1.000000)" 1e+01)
      (return "0, 0, 0"))))

(fn UGCBindRimLightIntensityChange
  ;[0.5059 1 0.6627] Is this a UGC experience?
  (bind _asfort_game_state_athena (Utilities|Casting|CastToFortGameStateAthena (Game|GetGameState)))
  (if (Class|FortGameStateAthena|IsSpatialGameplayExperience _asfort_game_state_athena)
    ;[0.5059 1 0.6627] If YES, then setup a callback for the value change on the character rim light intensity
    (bind _returnvalue (Fortnite|Utilities|AttemptGetPlayerStateforPawn self))
    (FortPlayerState|BindEventtoOnMutatorListUpdatedDelegate _returnvalue (EventDispatchers|CreateEvent :!event UpdateRimlightIntensity))))

(fn UGCInitRimlightIntensity
  (bind _asfort_game_state_athena (Utilities|Casting|CastToFortGameStateAthena (Game|GetGameState)))
  (if (Class|FortGameStateAthena|IsSpatialGameplayExperience _asfort_game_state_athena)
    (Variables|RimLight|SetRim%20Light%20Intensity (ScalableFloat|GetValueatLevelwithContext (Variables|RimLight|GetUGC%20RimLightIntensity)))))

(fn UpdateRimlightIntensity
  ;[0.5245 1 0.7717] If we are playing in a UGC experience then set the rim light intensity accordingly
  (RimLight|UGCInitRimlightIntensity)
  (Utilities|FlowControl|Sequence
    (:then_0
      ;[1 0.9178 0.4688] Set Physical Rimlight Intensity
      (bind _returnvalue (Development|GetConsoleVariableBoolValue "Athena.UsePhysicalRimlight"))
      (if _returnvalue
        (if (Pawn|IsLocallyControlled)
          (Utilities|FlowControl|Sequence
            (:then_0
              ;[0.6903 0.7536 1] Force a refresh of the actual light
              (PhysicalRimLight|ApplyPhysicalRimLightColorandIntensity (Variables|FortPlayerPawnAthena|GetRimLightComponent) (Variables|RimLight|GetRim%20Light%20Intensity)))
            (:then_1
              (RimLight|SetMaterialRimlightIntensity 0.0))))))
    (:then_1
      ;[0.5833 1 1] Set Material Rimlight Intensity
      (bind _returnvalue_2 (Development|GetConsoleVariableBoolValue "Athena.UseMaterialRimlight"))
      (if _returnvalue_2
        (bind _returnvalue_3 (Development|GetConsoleVariableBoolValue "Athena.UsePhysicalRimlight"))
        (if _returnvalue_3
          (if (Pawn|IsLocallyControlled)
            (RimLight|SetMaterialRimlightIntensity 0.0 :!label _label)
            (else
              _)))
        (RimLight|SetMaterialRimlightIntensity (Variables|RimLight|GetRim%20Light%20Intensity)))
      (goto _label))))

(fn ConstructionScript
  (|CallParentFunction)
  ;[0.02924 0.4632 1] store reference values for water FX determination so they are always up to date instead of being manually set
  (bind _returnvalue (Utilities|GetDataTableRow "/Game/Characters/Player/DataTables/CharacterAttributesPlayer.CharacterAttributesPlayer" "PlayerLoadout_AthenaPlayer"))
  (Variables|Default|SetPlayerAttributes_Speed_Walk (.SpeedWalk (Utilities|Struct|BreakFortPlayerPawnStats _returnvalue)))
  (Variables|Default|SetPlayerAttributes_Speed_Run (.SpeedRun (Utilities|Struct|BreakFortPlayerPawnStats _returnvalue)))
  (Variables|Default|SetPlayerAttributes_Speed_Sprint (.SpeedSprint (Utilities|Struct|BreakFortPlayerPawnStats _returnvalue))))
