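; Athena player pawn Blueprint holding helper functions for character effect meshes, hit audio, water state and replicated capsule size.
; Effect helpers add poseable or skeletal duplicates of body parts, and SlaveAMeshToTheBody leader-poses meshes to the top master or socket-attaches them.
; Shield breaks play through the hit-notify SoundLibrary with legacy fallback, notifying the locally viewed instigator.
; Water RTT enablement is cached from a global material parameter; remote clients apply replicated capsule height.
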(fn ArethewindandwaterRTTpassesenabled
  (if (not (Variables|Water|GetWas%20RTT%20Enabled%20Var%20Set))
    (bind _returnvalue (Rendering|Material|GetScalarParameterValue "/Game/Packages/Fortress_SharedMaterials/GlobalMaterialParameters/FortniteMaterialParameters.FortniteMaterialParameters" "Water Interaction Enabled" :!class /Script/Engine.KismetMaterialLibrary))
    (Variables|Water|SetIs%20RTT%20Enabled (> _returnvalue 0.5))
    (Variables|Water|SetWas%20RTT%20Enabled%20Var%20Set true))
  (return (Variables|Water|GetIs%20RTT%20Enabled)))

(fn CreateandDuplicateEffectPoseableSkeletalMesh
  (params (Material%20to%20Apply :type Material%20Interface%20Object%20Reference) (Empty%20MID%20Array :type Array%20of%20Material%20Instance%20Dynamic%20Object%20References) (TranslucentSortPriority :type Integer) (PoseableMesh :type Poseable%20Mesh%20Component%20Object%20Reference%20%28by%20ref%29) (CustomBodyType :type EFortCustomPartType%20Enum :default "Head") (PartMesh :type Skeletal%20Mesh%20Component%20Object%20Reference))
  (Utilities|CharacterDuplicateandSlavePoseableMeshes Material%20to%20Apply Empty%20MID%20Array PoseableMesh CustomBodyType PartMesh :!label _label
    (:Duplicate%20Skeletal%20Mesh%20Is%20Not%20Valid
      (bind _returnvalue (AddComponent|NODE_AddPoseableMeshComponent-1))
      (Variables|SetBy-RefVar PoseableMesh _returnvalue)
      (Rendering|SetTranslucentSortPriority _returnvalue TranslucentSortPriority)
      (goto _label))))

(fn CreateandDuplicateEffectSkeletalMeshesParent
  (params (DuplicatedSkeletalMeshComponent :type Skeletal%20Mesh%20Component%20Object%20Reference%20%28by%20ref%29) (Material%20to%20Apply :type Material%20Interface%20Object%20Reference) (Empty%20MID%20Array :type Array%20of%20Material%20Instance%20Dynamic%20Object%20References) (TranslucentSortPriority :type Integer) (MeshPart :type Skeletal%20Mesh%20Component%20Object%20Reference) (CustomBodyType :type EFortCustomPartType%20Enum :default "Head"))
  (Utilities|CharacterDuplicateandSlaveBodyMeshes :Skeletal%20Mesh%20Component%20Variable DuplicatedSkeletalMeshComponent :Mat%20Interface%20To%20Apply Material%20to%20Apply :Empty%20MID%20Array Empty%20MID%20Array :TranslucentSortPriority 0 :Ignore%20Base%20Material%20Input true :MeshPart MeshPart :CustomPartType CustomBodyType :!label _label
    (:Duplicate%20Skeletal%20Mesh%20Is%20Not%20Valid
      (bind _returnvalue (AddComponent|NODE_AddSkeletalMeshComponent-13))
      (Rendering|SetTranslucentSortPriority _returnvalue TranslucentSortPriority)
      (Variables|SetBy-RefVar DuplicatedSkeletalMeshComponent _returnvalue)
      (goto _label))))

(fn DisableWaterLevelTick
  (Variables|Water|SetTickWaterLevel false))

(fn GetInWater
  (return (Variables|VFX|GetIsInWater)))

(fn Melee_Effect_Color
  (return "0, 0, 0"))

(fn OnRep_CapsuleEnlarged
  (bind _capsulecomponent (Variables|Character|GetCapsuleComponent))
  (Networking|SwitchHasAuthority
    (:Remote
      (if (Math|Float|NearlyEqual(Float) (Variables|Default|GetCachedCapsuleHalfHeight))
        (Variables|Default|SetCachedCapsuleHalfHeight (Components|Capsule|GetUnscaledCapsuleHalfHeight _capsulecomponent)))
      (Components|Capsule|SetCapsuleHalfHeight _capsulecomponent (+ (Variables|Default|GetCachedCapsuleHalfHeight) (Variables|Default|GetReppedCapsuleExtraHalfHeight)) true))))

(fn PlayShieldDestroyedSound
  (params (GameplayCueParameters :type Gameplay%20Cue%20Parameters%20Structure%20%28by%20ref%29))
  (bind (_normalizedmagnitude _rawmagnitude _effectcontext _matchedtagname _originaltag _aggregatedsourcetags _aggregatedtargettags _location _normal _instigator _effectcauser) (Ability|GameplayCue|BreakGameplayCueParameters GameplayCueParameters))
  (bind _returnvalue (SoundLibrary|GetSoundLibraryComponent "(TagName=\"SoundLibrary.HitNotify\")"))
  (bind _instigator_2 (Variables|Actor|GetInstigator _effectcauser))
  (Utilities|FlowControl|Sequence
    (:then_0
      ;[0 0 0] New way of doing things with SoundLibrary, allows Experiences to specify if they want sounds to play
      (bind _asfort_hit_notify_sound_library_component (Utilities|Casting|CastToFortHitNotifySoundLibraryComponent _returnvalue))
      (Utilities|IsValid _returnvalue
        (:Is%20Valid
          (SoundLibrary|ConfigureContext _asfort_hit_notify_sound_library_component (Utilities|Struct|MakeFortHitNotifyContextSettings _effectcauser self) :!class /Script/FortniteGame.FortHitNotifySoundLibraryComponent)
          (bind _returnvalue_2 (SoundLibrary|PlayEvent _asfort_hit_notify_sound_library_component "(TagName=\"SoundLibrary.HitNotify.Event.ShieldBreak\")"))
          (if (not (.bSuccess (Utilities|Struct|BreakSoundLibraryPlayResult _returnvalue_2)))
            ;[0 0 0] Legacy - to be removed
            (Audio|SpawnSoundAttached (AudioGameplay|GetAudioforDamageEvent self _effectcauser (Variables|Sounds|GetSound_Shield_Destroyed_1P) (select (Ability|EffectContext|IsCriticalHit _effectcontext) (Variables|Sounds|GetSound_Shield_Destroyed_2D) (Variables|Sounds|GetSound_Shield_Destroyed_Crit_2D)) (Variables|Sounds|GetSound_Shield_Destroyed)) (Variables|Transformation|GetRootComponent))))))
    (:then_1
      (Utilities|IsValid _effectcauser
        (:Is%20Valid
          (bind _asplayer_pawn_athena (Utilities|Casting|CastToPlayerPawn_Athena _instigator_2)
            (:then
              (LocalVariables|SetLocalInstigatorPlayerPawn _asplayer_pawn_athena))
            (:CastFailed
              (bind _asfort_stand_in_player_pawn (Utilities|Casting|CastToFortStandInPlayerPawn _instigator_2))
              (LocalVariables|SetLocalInstigatorPlayerPawn _asfort_stand_in_player_pawn)))
          (if (Pawn|IsLocallyViewed (LocalVariables|GetLocalInstigatorPlayerPawn))
            (EventDispatchers|CallOnShieldBreak (LocalVariables|GetLocalInstigatorPlayerPawn) false)))))))

(fn ReInitFrontendPetDisplay
  (bind _cosmeticpetinstance (Variables|FortPlayerPawn|GetCosmeticPetInstance))
  (Utilities|IsValid _cosmeticpetinstance
    (:Is%20Valid
      (Class|FortPlayerPet|ReInitPetDisplay _cosmeticpetinstance))))

(fn SlaveAMeshToTheBody
  (params (Mesh :type Skeletal%20Mesh%20Component%20Object%20Reference) (Master :type Skeletal%20Mesh%20Component%20Object%20Reference))
  (bind _returnvalue (Transformation|GetAttachSocketName Master))
  (if (Utilities|Name|Equal(Name) _returnvalue "None")
    (LocalVariables|SetMasterToUse Master)
    ;[1 1 1] Use the master-most component
    (while (Utilities|IsValid (Variables|/Script/Engine.SkinnedMeshComponent|GetLeaderPoseComponent (LocalVariables|GetMasterToUse)) :!class /Script/Engine.KismetSystemLibrary)
      (LocalVariables|SetMasterToUse (Variables|/Script/Engine.SkinnedMeshComponent|GetLeaderPoseComponent (LocalVariables|GetMasterToUse))))
    (Transformation|SetRelativeTransform Mesh (Transformation|GetRelativeTransform (Variables|Character|GetMesh)) false false)
    (Components|SkinnedMesh|SetLeaderPoseComponent Mesh (LocalVariables|GetMasterToUse) false false)
    (Clothing|BindClothtoLeaderPoseComponent Mesh)
    (Rendering|SetVisibility Mesh false false)
    (else
      (Transformation|AttachComponentToComponent Mesh (Transformation|GetAttachParent Master) _returnvalue "SnapToTarget" "SnapToTarget" "SnapToTarget" true)
      (Rendering|SetVisibility Mesh false false))))

(fn ConstructionScript
  ;[1 1 1] Temporary hack because the outlander pawn is locked right now. It should be using RootComponent for the "component to attach" when playing sounds
  (Variables|Default|SetSound%20Player%20Hit (Variables|Transformation|GetRootComponent)))
