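; Generic parent Blueprint for Athena player pawns that adds local post-process setup and broadcasts vehicle, jump and crouch events.
; On BeginPlay, where cosmetic events can run, a locally controlled pawn creates a post-process component and stores it as PlayerPostProcessFX.
; Entering or exiting a vehicle sends a tagged gameplay event to the pawn and fires a matching dispatcher.
; Jumping and starting or ending a crouch share one dispatcher that lets Creative shooting targets react.
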(event EventBeginPlay
  (Networking|CanExecuteCosmeticEvents
    (:True
      (if (Pawn|IsLocallyControlled)
        (bind _returnvalue (AddComponent|NODE_AddPostProcessComponent-1))
        (Variables|Default|SetPlayerPostProcessFX _returnvalue)))))

(event Vehicle|EventOnEnteredVehicle
  ;[1 1 1] Enter and exit vehicle events.
  (|CallParentFunction)
  (Ability|SendGameplayEventtoActor self "(TagName=\"Event.Vehicle.Entered\")")
  (Default|CallOnEnteredVehicleDispatcher))

(event Vehicle|EventOnExitedVehicle
  ;[1 1 1] Enter and exit vehicle events.
  (|CallParentFunction)
  (Ability|SendGameplayEventtoActor self "(TagName=\"Event.Vehicle.Exit\")")
  (Default|CallOnExitedVehicleDispatcher))

(event Character|EventOnJumped
  ;[1 1 1] Shooting targets (B_ShootingTarget_Master) in Creative have a setting to react to the player jumping or crouching.  This dispatcher allows that behavior to function.
  (|CallParentFunction)
  (Default|CallShootingTargetReacttoJump self))

(event EventOnStartCrouch
  (params (HalfHeightAdjust :type Float%20%28single-precision%29) (ScaledHalfHeightAdjust :type Float%20%28single-precision%29))
  ;[1 1 1] Shooting targets (B_ShootingTarget_Master) in Creative have a setting to react to the player jumping or crouching.  This dispatcher allows that behavior to function.
  (|CallParentFunction HalfHeightAdjust ScaledHalfHeightAdjust)
  (Default|CallShootingTargetReacttoJump self))

(event EventOnEndCrouch
  (params (HalfHeightAdjust :type Float%20%28single-precision%29) (ScaledHalfHeightAdjust :type Float%20%28single-precision%29))
  ;[1 1 1] Shooting targets (B_ShootingTarget_Master) in Creative have a setting to react to the player jumping or crouching.  This dispatcher allows that behavior to function.
  (|CallParentFunction HalfHeightAdjust ScaledHalfHeightAdjust)
  (Default|CallShootingTargetReacttoJump self))

(fn ConstructionScript)
