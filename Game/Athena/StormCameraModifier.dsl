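; StormCameraModifier is a camera modifier that applies a stormy post-process look to the player's camera.
; Its BlueprintModifyPostProcess override returns a post-process settings struct weighted by the modifier's current alpha.
; The settings combine constant bloom and grading values, saturation and contrast read from its variables, and a storm cloud lookup texture from the world storm effects folder.
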(fn BlueprintModifyPostProcess
  (params (DeltaTime :type Float%20%28single-precision%29))
  (return (Variables|CameraModifier|GetAlpha) (Utilities|Struct|MakePostProcessSettings "BM_SOG" 6.5e+03 0.391 (Variables|Default|GetSaturation) (Variables|Default|GetContrast) 2.5 0.25 0.8 0.4 "/Game/Effects/Fort_Effects/World/Storms/T_LUT_Test_Cloudt.T_LUT_Test_Cloudt")))
